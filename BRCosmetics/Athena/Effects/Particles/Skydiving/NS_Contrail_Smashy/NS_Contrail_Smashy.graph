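# Niagara skydiving contrail for the Smashy cosmetic, combining confetti trails with bursts of smoke and confetti.
# ArmConfetti streams from skeletal-mesh sockets or bones, scaling rate and motion with paraglide lean.
# BurstSmoke and the center, low and main confetti emitters read CrossLama particles through attribute readers to trigger spawn-group bursts on death events.
# Confetti uses time-of-day tinting and brief brightness flashes, while front-end switches adjust some motion and sizing.
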NiagaraSystem /BRCosmetics/Athena/Effects/Particles/Skydiving/NS_Contrail_Smashy/NS_Contrail_Smashy
dependencies {
 usedDataInterfaces [
  &NiagaraDataInterfaceVectorCurve
  &NiagaraDataInterfaceCurve
  &NiagaraDataInterfaceParticleRead
  &NiagaraDataInterfaceArrayColor
  &NiagaraDataInterfaceArrayFloat
  &NiagaraDataInterfaceColorCurve
  &NiagaraDataInterfaceArrayInt32
  &NiagaraDataInterfaceArrayFloat3
  &NiagaraDataInterfaceMeshRendererInfo
  &NiagaraDataInterfaceVector2DCurve
  &NiagaraDataInterfaceSocketReader
  &NiagaraDataInterfaceSkeletalMesh
 ]
 usedDynamicInputs [
  &SimulationPosition
  &TransformVector
  /Niagara/DynamicInputs/Math/Clamp_Vector
  &NormalizeVector
 ]
 usedModules [
  &self:SystemSpawnScript.SystemScriptSource.SystemScriptGraph.NiagaraNodeAssignment_0.SetVariables_#1
  &self:SystemSpawnScript.SystemScriptSource.SystemScriptGraph.NiagaraNodeAssignment_1.SetVariables_#2
  &self:SystemSpawnScript.SystemScriptSource.SystemScriptGraph.NiagaraNodeAssignment_2.SetVariables_#3
  &SystemState
  &self:SystemSpawnScript.SystemScriptSource.SystemScriptGraph.NiagaraNodeAssignment_3.SetVariables_#4
  &Timeline
  /Game/Effects/Niagara/Modules/Contrails/FN_Contrail_Setup
  /Game/Effects/Niagara/Modules/Skydiving/FN_DivingAttributes/v5/FN_DivingAttributes
  &FN_Integrator
  &self:SystemSpawnScript.SystemScriptSource.SystemScriptGraph.NiagaraNodeAssignment_4.SetVariables_#5
  &self:SystemSpawnScript.SystemScriptSource.SystemScriptGraph.NiagaraNodeAssignment_5.SetVariables_#6
  &self:SystemSpawnScript.SystemScriptSource.SystemScriptGraph.NiagaraNodeAssignment_6.SetVariables_#7
 ]
}
emitters {
 ArmConfetti {
  emitterUpdateScript [
   &EmitterState(
    Inactive Response=::NewEnumerator1
    Life Cycle Mode=::NewEnumerator1
    Loop Behavior=::NewEnumerator1
    Loop Duration Mode=::NewEnumerator1
    UseLoopDelay=true
   )
   &SpawnRate(SpawnRate=&Lerp_Float(A=60, B=88, Alpha=$User.ParaGlideLeanAlpha), Use Spawn Probability=false)
  ]
  particleSpawnScript [
   &PartitionParticles(Partition Mode:ENiagara_PartitionMode_v2=::NewEnumerator5)
   &InitializeParticle(
    Color=&SelectLinearColorFromArray(
     Array Sampling Mode=::NewEnumerator0
     Direct Array Index=&ReturnExecIndex()
     Color Selection Array=DataInterface() {propertyValues {ColorData [<rgba 0.035 0.2349 1.599 1>, <rgba 0.4453 0.0491 2.2448 1>, <rgba 1.175 0.4351 3.1406 1>, <rgba 0 3.2318 5 1>]}}
     Random Seed=0
    )
    Lifetime=&RandomRangeFloat(Minimum=0.85, Maximum=1.5)
    Mesh Scale=&Multiply_VectorByFloat(
     Vector=&RandomRangeVector(Minimum=<1 1 1.2>, Maximum=<1.2 1.2 3.2>)
     Float=&RandomRangeFloat(Minimum=0.2, Maximum=0.3)
    )
    Mesh Scale Mode=::NewEnumerator4
    Sprite Rotation Mode=::NewEnumerator2
    Sprite Size=&Multiply_Vector2D_ByFloat(Float=&RandomRangeFloat(Minimum=5, Maximum=6), Vector2D=<1 1.5>)
    Sprite Size Mode=::NewEnumerator4
    UsePositionOffset=false
    Write Material Random=true
   ) {module=InitializeParticle001}
   &SkeletalMeshLocation(
    Bone / Socket Index=$Particles.PartitionParticles.Partition
    Bone Sampling Mode=::NewEnumerator4
    MaskSpawnGroup=true
    Orientation Sampling=::NewEnumerator1
    Position Sampling=::NewEnumerator1
    Sampled Position Offset=&FN_FrontEndVector3Modifier(
     Gameplay=&AddVectorToPosition(
      Position=<0 0 0>
      Vector=&RandomVector(Vector Scale=&Lerp_Float(A=20, B=10, Alpha=$User.ParaGlideLeanAlpha))
     )
     Front End=<0 -5 0>
    )
    Sampled Position Offset Coordinate Space:ENiagaraSkelMeshTransforms=::NewEnumerator1
    Skeletal Mesh=$System.Skeletal Mesh
   )
   &InitialMeshOrientation(
    Mesh Orientation Mode=::NewEnumerator6
    Orientation Quaternion:Quat4f=$Particles.SkeletalMeshLocation.SampledRotation
    Rotation=&RandomRangeVector(Minimum=<-0.1 -0.1 0>, Maximum=<0.1 0.1 1>)
    Use Rotation Vector=true
   )
   &AddVelocity(
    Cone Angle=&FN_FrontEndFloatModifier(Gameplay=5, Front End=20)
    Cone Axis=&FN_FrontEndVector3Modifier(Gameplay=$System.FN_Contrail_Setup.Direction Velocity, Front End=<0 0 1>)
    Module.UseConeFalloff=false
    Module.UseVelDistribution=false
    Override Seed=false
    Rotation Coordinate Space=::World
    Velocity Mode=::NewEnumerator2
    Velocity Speed=&Lerp_Float(
     A=&RandomRangeFloat(Minimum=50, Maximum=70)
     B=&RandomRangeFloat(Minimum=200, Maximum=250)
     Alpha=$User.ParaGlideLeanAlpha
    )
   ) {module=AddVelocity004}
   &Color(Color=&MultiplyLinearColor(
     Linear Color=$Particles.Color
     Scale Factor=&MakeVector4_FromLinearColor(LinearColor=&FNTimeOfDayColor(
       Night Color=<rgba 0.2706 0.2383 0.4948 1>
       Morning Color=<rgba 1 0.7262 0.6963 1>
       Day Color=<rgba 1 0.9315 0.9012 1>
       Evening Color=<rgba 1 0.9069 0.7687 1>
      ))
    ))
  ]
  particleUpdateScript [
   &ParticleState()
   &ScaleSpriteSize(Uniform Curve Sprite Scale=DataInterface() {propertyValues {Curve {keys [
        {interpMode=RCIM_Cubic, leaveTangent=12.0846, tangentMode=RCTM_Break, time=0, value=0}
        {interpMode=RCIM_Cubic, tangentMode=RCTM_Break, time=0.1142, value=1.0195}
        {arriveTangent=-0.3181, interpMode=RCIM_Cubic, leaveTangent=-0.3181, tangentMode=RCTM_User, time=0.808, value=1.0348}
        {arriveTangent=-7.8683, tangentMode=RCTM_User, time=1, value=0}
       ]}}})
   &LinearForce(Force=&FN_FrontEndVector3Modifier(
     Gameplay=&Multiply_VectorByFloat(Vector=&Multiply_Vector(A=$System.FN_Contrail_Setup.Direction Vector, B=<0 0 1>), Float=-300)
     Front End=&Multiply_VectorByFloat(Vector=&Multiply_Vector(A=$System.FN_Contrail_Setup.Direction Vector, B=<0 1 1>), Float=-550)
    ))
   &PointForce(
    Force Falloff Distance:NiagaraFloat=&FN_FrontEndFloatModifier(Gameplay=400, Front End=0)
    Force Strength=&FN_FrontEndFloatModifier(Gameplay=&Lerp_Float(A=500, B=200, Alpha=$User.ParaGlideLeanAlpha), Front End=0)
    Module.Use Falloff=true
    Module.UseLinearForce=false
   )
   &AerodynamicDrag(
    Aerodynamic Drag=0.25
    Drag Shape Curve=DataInterface() {propertyValues {Curve {keys [
        {interpMode=RCIM_Cubic, tangentMode=RCTM_User, time=0, value=0.1}
        {arriveTangent=2.0799, interpMode=RCIM_Cubic, leaveTangent=2.0799, tangentMode=RCTM_User, time=1, value=1}
       ]}}}
    Ignore Mass=true
    Lift Shape Curve=DataInterface() {propertyValues {Curve {keys [
        {interpMode=RCIM_Cubic, tangentMode=RCTM_User, time=0, value=0}
        {interpMode=RCIM_Cubic, tangentMode=RCTM_User, time=0.5, value=1}
        {tangentMode=RCTM_User, time=1, value=0}
       ]}}}
    Rotation Strength=20
    Surface Normal Axis:Vector3f=<0 0 1>
    Use Linear Drag=true
    Use Rotational Drag=true
   )
   &CurlNoiseForce(
    Bias Noise=false
    Noise Frequency=20
    Noise Strength=&Lerp_Float(A=&FN_FrontEndFloatModifier(Gameplay=1000, Front End=250), B=250, Alpha=$User.ParaGlideLeanAlpha)
    Pan Noise=false
   )
   &SolveForcesAndVelocity(Clamp Velocity=false, Limit Acceleration=false)
   &AlignSpriteToMeshOrientation(SetAlignment=true, SetFacing=true)
   /SetVariables(Particles.SubImageIndex=&FloatFromCurve(FloatCurve=DataInterface() {propertyValues {Curve {keys [<at 0 0>,<at 1 1>]}}}, Scale Curve=36)) {isSetParameters=true}
   &ScaleColor(
    Scale Mode=1
    Scale RGBA=&MakeVector4_FromLinearColor(LinearColor=&FNTimeOfDayColor(
      Night Color=&ColorFromCurve(ColorCurve=DataInterface() {propertyValues {
         AlphaCurve {keys [<at 0 1 RCIM_Cubic>]}
         BlueCurve {keys [<at 0 1 RCIM_Cubic>,<at 0.4559 1>,<at 0.4752 250>,<at 0.4979 1>,<at 1 1 RCIM_Cubic>]}
         GreenCurve {keys [<at 0 1 RCIM_Cubic>,<at 0.4559 1>,<at 0.4752 250>,<at 0.4979 1>,<at 1 1 RCIM_Cubic>]}
         RedCurve {keys [<at 0 1 RCIM_Cubic>,<at 0.4559 1>,<at 0.4752 250>,<at 0.4979 1>,<at 1 1 RCIM_Cubic>]}
        }})
      Morning Color=&ColorFromCurve(ColorCurve=DataInterface() {propertyValues {
         AlphaCurve {keys [<at 0 1 RCIM_Cubic>]}
         BlueCurve {keys [<at 0 1 RCIM_Cubic>,<at 0.4559 1>,<at 0.4752 250>,<at 0.4979 1>,<at 1 1 RCIM_Cubic>]}
         GreenCurve {keys [<at 0 1 RCIM_Cubic>,<at 0.4559 1>,<at 0.4752 250>,<at 0.4979 1>,<at 1 1 RCIM_Cubic>]}
         RedCurve {keys [<at 0 1 RCIM_Cubic>,<at 0.4559 1>,<at 0.4752 250>,<at 0.4979 1>,<at 1 1 RCIM_Cubic>]}
        }})
      Day Color=&ColorFromCurve(ColorCurve=DataInterface() {propertyValues {
         AlphaCurve {keys [<at 0 1 RCIM_Cubic>]}
         BlueCurve {keys [<at 0 1 RCIM_Cubic>,<at 0.4559 1>,<at 0.4752 250>,<at 0.4979 1>,<at 1 1 RCIM_Cubic>]}
         GreenCurve {keys [<at 0 1 RCIM_Cubic>,<at 0.4559 1>,<at 0.4752 250>,<at 0.4979 1>,<at 1 1 RCIM_Cubic>]}
         RedCurve {keys [<at 0 1 RCIM_Cubic>,<at 0.4559 1>,<at 0.4752 250>,<at 0.4979 1>,<at 1 1 RCIM_Cubic>]}
        }})
      Evening Color=&ColorFromCurve(ColorCurve=DataInterface() {propertyValues {
         AlphaCurve {keys [<at 0 1 RCIM_Cubic>]}
         BlueCurve {keys [<at 0 1 RCIM_Cubic>,<at 0.4559 1>,<at 0.4752 250>,<at 0.4979 1>,<at 1 1 RCIM_Cubic>]}
         GreenCurve {keys [<at 0 1 RCIM_Cubic>,<at 0.4559 1>,<at 0.4752 250>,<at 0.4979 1>,<at 1 1 RCIM_Cubic>]}
         RedCurve {keys [<at 0 1 RCIM_Cubic>,<at 0.4559 1>,<at 0.4752 250>,<at 0.4979 1>,<at 1 1 RCIM_Cubic>]}
        }})
     ))
    ScaleRGBA=true
   )
  ]
  rendererClasses [&NiagaraSpriteRendererProperties]
  renderers [{rendererClass=&NiagaraSpriteRendererProperties}]
 }
 BurstSmoke {
  emitterUpdateScript [
   &EmitterState(Life Cycle Mode=::NewEnumerator1, Loop Duration=$Engine.DeltaTime, Recalculate Duration Each Loop=true, UseLoopDelay=false)
   /SetVariables(Emitter.BoolDeathEvent:NiagaraBool=&self:ScratchDynamicInput_17(Particle Attribute Reader:NiagaraDataInterfaceParticleRead=DataInterface() {propertyValues {EmitterBinding {bindingMode=Other, emitterName=CrossLama}}})) {isSetParameters=true}
   /Niagara/Modules/Emitter/SpawnByTrigger(Spawn Count=3, Trigger:NiagaraBool=$Emitter.BoolDeathEvent, TriggerIntervalSwitch=false)
   &SpawnByTrigger(Spawn Count=3, Spawn Group=1, Trigger:NiagaraBool=$Emitter.BoolDeathEvent, TriggerIntervalSwitch=false) {module=SpawnByTrigger001}
   &SpawnByTrigger(Spawn Count=10, Spawn Group=2, Trigger:NiagaraBool=$Emitter.BoolDeathEvent, TriggerIntervalSwitch=false) {module=SpawnByTrigger002}
   &SpawnByTrigger(Spawn Group=3, Trigger:NiagaraBool=$Emitter.BoolDeathEvent, TriggerIntervalSwitch=false) {module=SpawnByTrigger003}
  ]
  particleSpawnScript [
   &InitializeParticle(
    Color=&MaskLinearColor(
     Mask Source=0
     Mask Parameter=?
     Mask=2
     Masked Linear Color=&SelectLinearColorFromArray(
      Array Sampling Mode=::NewEnumerator0
      Direct Array Index=?
      Color Selection Array=DataInterface() {propertyValues {ColorData [<rgba 2.1633 1.026 5 1>, <rgba 0.5156 0.53 3 1>, <rgba 0.101 0.4334 1 1.1>]}}
      Random Seed=0
     )
     Passthrough Linear Color=&SelectLinearColorFromArray(
      Array Sampling Mode=::NewEnumerator0
      Direct Array Index=?
      Color Selection Array=DataInterface() {propertyValues {ColorData [<rgba 0.5128 0.3162 0.5781 0.6>, <rgba 0.5622 0.5465 0.8615 0.4>, <rgba 0.3397 0.3734 0.6458 0.8>]}}
      Random Seed=0
     )
    )
    Lifetime Max=&MergeHelper_FloatSwitch(0=0.8, 1=0.6, 2=0.7, 3=0.15, Index=&SpawnGroup(), Num=::NewEnumerator2)
    Lifetime Min:NiagaraFloat=&MergeHelper_FloatSwitch(0=0.6, 1=0.3, 2=0.5, 3=0.15, Index=&SpawnGroup(), Num=::NewEnumerator2)
    Lifetime Mode=::NewEnumerator1
    Mesh Scale Mode=::NewEnumerator3
    Mesh Uniform Scale=&FN_FrontEndFloatModifier(Gameplay=7, Front End=4)
    Sprite Rotation Mode=::NewEnumerator2
    Sprite Size Mode=::NewEnumerator3
    Uniform Sprite Size=&MergeHelper_FloatSwitch(
     0=&FN_FrontEndFloatModifier(
      Gameplay=&RandomRangeFloat(Minimum=100, Maximum=120)
      Front End=&RandomRangeFloat(Minimum=100, Maximum=120)
     )
     1=&FN_FrontEndFloatModifier(
      Gameplay=&RandomRangeFloat(Minimum=300, Maximum=400)
      Front End=&RandomRangeFloat(Minimum=300, Maximum=400)
     )
     2=&FN_FrontEndFloatModifier(
      Gameplay=&RandomRangeFloat(Minimum=80, Maximum=100)
      Front End=&RandomRangeFloat(Minimum=80, Maximum=100)
     )
     3=?
     Index=&SpawnGroup()
     Num=::NewEnumerator1
    )
    UsePositionOffset=false
    Write Material Random=true
   )
   &CameraOffset(Camera Offset Amount=1)
   /SetVariables(Particles.VisibilityTag=&SelectIntFromArray(Direct Array Index=&SpawnGroup(), Int Selection Array=DataInterface() {propertyValues {IntData [0,0,1,2]}})) {isSetParameters=true}
   &SocketLocation(
    Orientation Sampling=::NewEnumerator1
    SamplePosOffset=true
    Sampled Position Offset=&RandomVector(Vector Scale=&MergeHelper_FloatSwitch(0=50, 1=0, 2=0, 3=0, Index=&SpawnGroup(), Num=::NewEnumerator2))
    Socket Index=4
    Socket Reader=$System.Socket Reader
    Socket Sampling Mode=::NewEnumerator3
    VelocitySpeedLimit=false
   )
   &AddVelocity(
    Module.Use Falloff Distance=false
    Override Seed=false
    Velocity Mode=::NewEnumerator1
    Velocity Origin:Vector3f=$Output.SocketLocation.SampledPosition
    Velocity Speed=&MergeHelper_FloatSwitch(
     0=0
     1=0
     2=&FN_FrontEndFloatModifier(
      Gameplay=&RandomRangeFloat(Minimum=650, Maximum=500)
      Front End=&RandomRangeFloat(Minimum=400, Maximum=500)
     )
     3=0
     Index=&SpawnGroup()
     Num=::NewEnumerator2
    )
   ) {module=AddVelocity001}
   &AddVelocity(
    Velocity=&MergeHelper_VectorSwitch(
     0=&RandomRangeVector(Minimum=<0 0 200>, Maximum=<0 0 300>)
     1=/Niagara/DynamicInputs/Vectors/RotateVector(
      Vector To Rotate:Vector3f=<1 0 0>
      Yaw:NiagaraFloat=&ReturnNormalizedExecIndex(Normalized Index Scale=360, Normalized Index Includes One=false)
      Pitch:NiagaraFloat=0
      Roll:NiagaraFloat=0
      Rotation Rate:NiagaraFloat=1
     )
     2=<0 0 0>
     3=?
     Index=&SpawnGroup()
     Num=::NewEnumerator1
    )
    Velocity Speed Scale=&MergeHelper_FloatSwitch(
     0=1
     1=&FN_FrontEndFloatModifier(
      Gameplay=&RandomRangeFloat(Minimum=350, Maximum=500)
      Front End=&RandomRangeFloat(Minimum=350, Maximum=500)
     )
     2=0
     3=0
     Index=&SpawnGroup()
     Num=::NewEnumerator2
    )
   ) {module=AddVelocity003}
   &InheritVelocity(
    Inherited Velocity Amount Scale=<0 0 -1>
    Inherited Velocity Speed Limit:NiagaraFloat=&MergeHelper_FloatSwitch(0=400, 1=0, 2=300, 3=0, Index=&SpawnGroup(), Num=::NewEnumerator2)
    Speed Limit=true
    UseSpeedThreshold=false
    Velocity Destination Space:ENiagaraCoordinateSpace=::World
   )
   /SetVariables(
    Particles.FlareColor=<rgba 0.8687 0.2842 2 0.7>
    Particles.FlarePosition:NiagaraPosition=&MakeCustomVectorFromBool(
     Bool=$User.bIsFrontEnd
     True Vector=Linked() {linkedVariable {name=Particles.Position, type=NiagaraPosition}}
     False Vector=$Output.SocketLocation.SampledPosition
    )
   ) {isSetParameters=true}
   /SetVariables(Particles.DMPFlare:Vector4f=&MakeVector4(
     X=&RandomRangeFloat(Minimum=0, Maximum=10)
     Y=&RandomRangeFloat(Minimum=2.5, Maximum=4.5)
     Z=&RandomRangeFloat(Minimum=0.8, Maximum=1.6)
     W=0
    )) {isSetParameters=true}
  ]
  particleUpdateScript [
   &ParticleState()
   &ScaleMeshSize(Scale Factor=&VectorFromFloat(Value=&FloatFromCurve(FloatCurve=DataInterface() {propertyValues {Curve {
         defaultValue=0
         keys [
          {arriveTangent=4.1785, interpMode=RCIM_Cubic, leaveTangent=4.1785, tangentMode=RCTM_User, time=0, value=0}
          {arriveTangent=1.1179, interpMode=RCIM_Cubic, leaveTangent=1.1179, tangentMode=RCTM_User, time=0.2042, value=0.6228}
          {arriveTangent=0.0224, interpMode=RCIM_Cubic, leaveTangent=0.0224, tangentMode=RCTM_User, time=1, value=1}
         ]
        }}})))
   &SpriteRotationRate(Rotation Rate=&MaskFloat(
     Mask Source=0
     Mask Parameter=?
     Mask=2
     Masked Float=&RandomRangeFloat(Minimum=-90, Maximum=90)
     Passthrough Float=0
    ))
   &ScaleSpriteSize(Uniform Curve Sprite Scale=DataInterface() {propertyValues {Curve {keys [
        {interpMode=RCIM_Cubic, leaveTangent=2.1273, tangentMode=RCTM_Break, time=0, value=0.2}
        {arriveTangent=0.0779, interpMode=RCIM_Cubic, tangentMode=RCTM_Break, time=1, value=1}
       ]}}})
   &Drag(
    Drag=&MergeHelper_FloatSwitch(0=3, 1=1, 2=4, 3=0, Index=&SpawnGroup(), Num=::NewEnumerator2)
    Use Linear Drag=true
    Use Rotational Drag=true
   )
   &GravityForce(Gravity=&MaskVector(
     Mask Source=0
     Mask Parameter=?
     Mask=3
     Masked Vector=<0 0 0>
     Passthrough Vector=&FN_FrontEndVector3Modifier(Gameplay=<0 0 0>, Front End=<0 0 980>)
    ))
   &SolveForcesAndVelocity(Clamp Velocity=false, Limit Acceleration=false)
   &DynamicMaterialParameters(
    Index 0 Param 1=&FloatFromCurve(FloatCurve=DataInterface() {propertyValues {Curve {keys [<at 0 0 RCIM_Cubic>,<at 1.0011 0.2>]}}})
    Index 0 Param 2=&FloatFromCurve(FloatCurve=DataInterface() {propertyValues {Curve {keys [
         {arriveTangent=0.0453, interpMode=RCIM_Cubic, tangentMode=RCTM_Break, time=0, value=-1}
         {arriveTangent=0.0604, interpMode=RCIM_Cubic, leaveTangent=0.0604, tangentMode=RCTM_User, time=1, value=-1}
        ]}}})
    Index 0 Param 3=&FloatFromCurve(FloatCurve=DataInterface() {propertyValues {Curve {keys [{interpMode=RCIM_Cubic, leaveTangent=4.6779, time=0.2851, value=0.4}]}}})
    Index 1 Param 1=&FloatFromCurve(
     FloatCurve=DataInterface() {propertyValues {Curve {keys [<at 0 0>,<at 0.25 1>]}}}
     Scale Curve=&RandomRangeFloat(Minimum=2, Maximum=1)
    )
    Index 1 Param 2=&FloatFromCurve(
     FloatCurve=DataInterface() {propertyValues {Curve {keys [
         {interpMode=RCIM_Cubic, leaveTangent=3.7422, tangentMode=RCTM_Break, time=0, value=0.2}
         {interpMode=RCIM_Cubic, leaveTangent=0.9066, time=1, value=2}
        ]}}}
     Scale Curve=&RandomRangeFloat(Minimum=0.5, Maximum=1)
    )
    Index 1 Param 3=&FloatFromCurve(FloatCurve=DataInterface() {propertyValues {Curve {keys [
         {interpMode=RCIM_Cubic, leaveTangent=2.8445, tangentMode=RCTM_Break, time=0, value=0.2}
         {interpMode=RCIM_Cubic, leaveTangent=63.2978, time=0.25, value=1}
        ]}}})
    Index 1 Param 4=&FloatFromCurve(
     FloatCurve=DataInterface() {propertyValues {Curve {keys [<at 0 1>,<at 1 0.8>]}}}
     Scale Curve=&RandomRangeFloat(Minimum=0.5, Maximum=1)
    )
    Index 2 Param 1=&FloatFromCurve(FloatCurve=DataInterface() {propertyValues {Curve {keys [
         {interpMode=RCIM_Cubic, leaveTangent=11.9967, tangentMode=RCTM_Break, time=0.1, value=0}
         {arriveTangent=0.32, interpMode=RCIM_Cubic, tangentMode=RCTM_Break, time=1, value=1}
        ]}}})
    Index 2 Param 2=$Particles.MaterialRandom
    Index 2 Param 3=&FloatFromCurve(FloatCurve=DataInterface() {propertyValues {Curve {keys [
         {interpMode=RCIM_Cubic, tangentMode=RCTM_Break, time=0, value=1}
         {interpMode=RCIM_Cubic, leaveTangent=11.1233, tangentMode=RCTM_Break, time=1, value=2}
        ]}}})
    Index 2 Param 4=0.75
    Module.Param0WriteEnabled=true
    Module.Param0WriteEnabled1=true
    Module.Param0WriteEnabled2=true
    Module.Param1WriteEnabled=false
    Module.Param1WriteEnabled1=true
    Module.Param1WriteEnabled2=true
    Module.Param2WriteEnabled=true
    Module.Param2WriteEnabled1=true
    Module.Param2WriteEnabled2=true
    Module.Param3WriteEnabled=true
    Module.Param3WriteEnabled1=true
    Module.Param3WriteEnabled2=true
    Write Parameter Index 1=true
    Write Parameter Index 2=true
   )
   &ScaleColor(
    Scale Alpha=&FloatFromCurve(FloatCurve=DataInterface() {propertyValues {Curve {keys [
         {interpMode=RCIM_Cubic, tangentMode=RCTM_Break, time=0, value=0.5}
         {interpMode=RCIM_Cubic, tangentMode=RCTM_Break, time=0.3, value=1}
         {interpMode=RCIM_Cubic, tangentMode=RCTM_Break, time=1, value=0}
        ]}}})
    Scale RGB=&MaskVector(
     Mask Source=0
     Mask Parameter=?
     Mask=2
     Masked Vector=&VectorFromFloat(Value=&Lerp_Float(
       A=1
       B=2
       Alpha=&FloatFromCurve(
        FloatCurve=DataInterface() {propertyValues {Curve {keys [<at 0 0 RCIM_Constant>,<at 0.8 1>]}}}
        CurveIndex=&RandomRangeFloat(Minimum=0, Maximum=1)
       )
      ))
     Passthrough Vector=<1 1 1>
    )
    ScaleA=true
    ScaleRGB=true
   )
   /SetVariables(Particles.FlareColor=&MultiplyLinearColor(
     Linear Color=$Particles.FlareColor
     Scale Factor=&MakeVector4(
      X=1
      Y=1
      Z=1
      W=&FloatFromCurve(FloatCurve=DataInterface() {propertyValues {Curve {
          defaultValue=1
          keys [{interpMode=RCIM_Cubic, tangentMode=RCTM_User, time=0, value=1}, {interpMode=RCIM_Cubic, tangentMode=RCTM_User, time=1, value=0}]
         }}})
     )
    )) {isSetParameters=true}
  ]
  rendererClasses [&NiagaraSpriteRendererProperties, &NiagaraMeshRendererProperties]
  renderers [
   {rendererClass=&NiagaraSpriteRendererProperties}
   {rendererClass=&NiagaraSpriteRendererProperties, rendererIndex=1}
   {rendererClass=&NiagaraMeshRendererProperties, rendererIndex=2}
  ]
 }
 ConfettiCenter {
  emitterUpdateScript [
   /SetVariables(Emitter.BoolDeathEvent:NiagaraBool=&self:ScratchDynamicInput_15(Particle Attribute Reader:NiagaraDataInterfaceParticleRead=DataInterface() {propertyValues {EmitterBinding {bindingMode=Other, emitterName=CrossLama}}})) {isSetParameters=true, module=SetVariables_#8}
   &EmitterState(Life Cycle Mode=::NewEnumerator1, Loop Delay=0.5, Loop Duration=$Engine.DeltaTime, Recalculate Duration Each Loop=true, UseLoopDelay=false)
   &SpawnByTrigger(
    Spawn Count:NiagaraInt32=&RandomRangeInt(Minimum=55, Maximum=65)
    Trigger:NiagaraBool=$Emitter.BoolDeathEvent
    TriggerIntervalSwitch=false
   )
  ]
  particleSpawnScript [
   &InitializeParticle(
    Color=&SelectLinearColorFromArray(
     Array Sampling Mode=::NewEnumerator0
     Direct Array Index=&ReturnExecIndex()
     Color Selection Array=DataInterface() {propertyValues {ColorData [<rgba 0.035 0.2349 1.599 1>, <rgba 0.4453 0.0491 2.2448 1>, <rgba 1.175 0.4351 3.1406 1>, <rgba 0 3.2318 5 1>]}}
     Random Seed=0
    )
    Lifetime Mode=::NewEnumerator1
    Mesh Scale=&Multiply_VectorByFloat(
     Vector=&RandomRangeVector(Minimum=<1 1 1>, Maximum=<1.05 1.05 3>)
     Float=&RandomRangeFloat(Minimum=0.35, Maximum=0.5)
    )
    Mesh Scale Mode=::NewEnumerator4
    Sprite Size=&Multiply_Vector2D_ByFloat(Float=&RandomRangeFloat(Minimum=5, Maximum=10), Vector2D=<1 1.5>)
    Sprite Size Mode=::NewEnumerator4
    UsePositionOffset=false
    Write Material Random=true
   ) {module=InitializeParticle001}
   &SocketLocation(
    Orientation Sampling=::NewEnumerator1
    Position Sampling=::NewEnumerator1
    SamplePosOffset=true
    Sampled Position Coordinate Space:ENiagaraCoordinateSpace=::Local
    Sampled Position Offset=&RandomVector(Vector Scale=5)
    Socket Index=4
    Socket Reader=$System.Socket Reader
    Socket Sampling Mode=::NewEnumerator3
    VelocitySpeedLimit=false
   )
   /SetVariables(Particles.Position:Vector3f=$Output.SocketLocation.SampledPosition) {isSetParameters=true, module=SetVariables_#9}
   &InitialMeshOrientation(Use Rotation Vector=false)
   &ShapeLocation(
    MaskSpawnGroup=false
    Override Seed=false
    Shape Origin=$Particles.Position
    Sphere Radius=5
    Sphere Surface Distribution=0.72
    UseOwnerScale=true
   )
   &AddVelocity(
    Module.Use Falloff Distance=false
    Origin Offset=<0 0 -1>
    Override Seed=false
    Velocity Mode=::NewEnumerator1
    Velocity Origin=$Particles.Position
    Velocity Speed=&RandomRangeFloat(Minimum=5, Maximum=10)
   ) {module=AddVelocity001}
   &CurlNoiseForce(
    Bias Noise=false
    Noise Frequency=10
    Noise Strength=&RandomRangeFloat(Minimum=25, Maximum=50)
    Pan Noise=true
    Pan Noise Field=<0.2016 0 0.5008>
   ) {module=CurlNoiseForce001}
   &AddVelocity(
    Cone Axis=&Multiply_VectorByFloat(Vector=$System.FN_Contrail_Setup.Direction Vector, Float=-1)
    Module.UseConeFalloff=false
    Module.UseVelDistribution=false
    Override Seed=false
    Velocity Mode=::NewEnumerator2
    Velocity Speed=&RandomRangeFloat(Minimum=350, Maximum=888)
   ) {module=AddVelocity003}
   &InheritVelocity(
    Inherited Velocity Amount Scale=<0 0 -1>
    Inherited Velocity Speed Limit=300
    Speed Limit=true
    UseSpeedThreshold=false
    Velocity Destination Space:ENiagaraCoordinateSpace=::World
   )
   &ApplyInitialForces()
   &Color(Color=&MultiplyLinearColor(
     Linear Color=$Particles.Color
     Scale Factor=&MakeVector4_FromLinearColor(LinearColor=&FNTimeOfDayColor(
       Night Color=<rgba 0.2706 0.2383 0.4948 1>
       Morning Color=<rgba 1 0.7262 0.6963 1>
       Day Color=<rgba 1 0.9315 0.9012 1>
       Evening Color=<rgba 1 0.9069 0.7687 1>
      ))
    ))
  ]
  particleUpdateScript [
   &ParticleState()
   &SpriteRotationRate(Rotation Rate=&Multiply_Float(
     A=&SelectFloatFromArray(
      Array Sampling Mode=::NewEnumerator0
      Direct Array Index=?
      Float Selection Array=DataInterface() {propertyValues {FloatData [-300,300]}}
      Random Seed=0
     )
     B=&RandomRangeFloat(Minimum=-5, Maximum=5)
    ))
   &ScaleSpriteSize(Scale Factor=&Vector2DFromFloat(Value=&FloatFromCurve(
      FloatCurve=DataInterface() {propertyValues {Curve {keys [
          {arriveTangent=-0.4596, interpMode=RCIM_Cubic, leaveTangent=-0.4596, tangentMode=RCTM_User, time=0.8, value=1}
          {arriveTangent=-4.9031, interpMode=RCIM_Cubic, leaveTangent=-4.9031, tangentMode=RCTM_User, time=1, value=0}
         ]}}}
      Scale Curve=&RandomRangeFloat(Minimum=1, Maximum=1, Randomness Mode=::NewEnumerator2, Random Seed=?)
     ))) {module=ScaleSpriteSize001}
   &Drag(Drag=50, Use Linear Drag=true, Use Rotational Drag=true) {module=Drag002}
   /Niagara/Modules/Update/Forces/CurlNoiseForce(Noise Frequency=50, Noise Strength=&RandomRangeFloat(Minimum=100, Maximum=150), Pan Noise=false)
   &AerodynamicDrag(
    Aerodynamic Drag=0.2
    Drag Shape Curve=DataInterface() {propertyValues {Curve {keys [
        {interpMode=RCIM_Cubic, tangentMode=RCTM_User, time=0, value=0.1}
        {arriveTangent=2.0799, interpMode=RCIM_Cubic, leaveTangent=2.0799, tangentMode=RCTM_User, time=1, value=1}
       ]}}}
    Ignore Mass=true
    Lift Shape Curve=DataInterface() {propertyValues {Curve {keys [
        {interpMode=RCIM_Cubic, tangentMode=RCTM_User, time=0, value=0}
        {interpMode=RCIM_Cubic, tangentMode=RCTM_User, time=0.5, value=1}
        {tangentMode=RCTM_User, time=1, value=0}
       ]}}}
    Surface Normal Axis:Vector3f=<0 0 1>
    Use Linear Drag=true
    Use Rotational Drag=true
   )
   &GravityForce(Gravity=<0 0 980>)
   &SolveForcesAndVelocity(Clamp Velocity=false, Limit Acceleration=false)
   &AlignSpriteToMeshOrientation(SetAlignment=true, SetFacing=true)
   /SetVariables(Particles.SubImageIndex=&FloatFromCurve(FloatCurve=DataInterface() {propertyValues {Curve {keys [<at 0 0>,<at 1 1>]}}}, Scale Curve=36)) {isSetParameters=true, module=SetVariables_#10}
   &KillParticles(Kill Particles=&SetBoolByFloatComparison(A=&VectorLength(Vector=$Particles.Velocity), B=1, Comparison Type=::NewEnumerator6))
   &ScaleColor(
    Scale Mode=1
    Scale RGBA=&MakeVector4_FromLinearColor(LinearColor=&FNTimeOfDayColor(
      Night Color=&ColorFromCurve(ColorCurve=DataInterface() {propertyValues {
         AlphaCurve {keys [<at 0 1 RCIM_Cubic>]}
         BlueCurve {keys [<at 0 1 RCIM_Cubic>,<at 0.4559 1>,<at 0.4752 250>,<at 0.4979 1>,<at 1 1 RCIM_Cubic>]}
         GreenCurve {keys [<at 0 1 RCIM_Cubic>,<at 0.4559 1>,<at 0.4752 250>,<at 0.4979 1>,<at 1 1 RCIM_Cubic>]}
         RedCurve {keys [<at 0 1 RCIM_Cubic>,<at 0.4559 1>,<at 0.4752 250>,<at 0.4979 1>,<at 1 1 RCIM_Cubic>]}
        }})
      Morning Color=&ColorFromCurve(ColorCurve=DataInterface() {propertyValues {
         AlphaCurve {keys [<at 0 1 RCIM_Cubic>]}
         BlueCurve {keys [<at 0 1 RCIM_Cubic>,<at 0.4559 1>,<at 0.4752 250>,<at 0.4979 1>,<at 1 1 RCIM_Cubic>]}
         GreenCurve {keys [<at 0 1 RCIM_Cubic>,<at 0.4559 1>,<at 0.4752 250>,<at 0.4979 1>,<at 1 1 RCIM_Cubic>]}
         RedCurve {keys [<at 0 1 RCIM_Cubic>,<at 0.4559 1>,<at 0.4752 250>,<at 0.4979 1>,<at 1 1 RCIM_Cubic>]}
        }})
      Day Color=&ColorFromCurve(ColorCurve=DataInterface() {propertyValues {
         AlphaCurve {keys [<at 0 1 RCIM_Cubic>]}
         BlueCurve {keys [<at 0 1 RCIM_Cubic>,<at 0.4559 1>,<at 0.4752 250>,<at 0.4979 1>,<at 1 1 RCIM_Cubic>]}
         GreenCurve {keys [<at 0 1 RCIM_Cubic>,<at 0.4559 1>,<at 0.4752 250>,<at 0.4979 1>,<at 1 1 RCIM_Cubic>]}
         RedCurve {keys [<at 0 1 RCIM_Cubic>,<at 0.4559 1>,<at 0.4752 250>,<at 0.4979 1>,<at 1 1 RCIM_Cubic>]}
        }})
      Evening Color=&ColorFromCurve(ColorCurve=DataInterface() {propertyValues {
         AlphaCurve {keys [<at 0 1 RCIM_Cubic>]}
         BlueCurve {keys [<at 0 1 RCIM_Cubic>,<at 0.4559 1>,<at 0.4752 250>,<at 0.4979 1>,<at 1 1 RCIM_Cubic>]}
         GreenCurve {keys [<at 0 1 RCIM_Cubic>,<at 0.4559 1>,<at 0.4752 250>,<at 0.4979 1>,<at 1 1 RCIM_Cubic>]}
         RedCurve {keys [<at 0 1 RCIM_Cubic>,<at 0.4559 1>,<at 0.4752 250>,<at 0.4979 1>,<at 1 1 RCIM_Cubic>]}
        }})
     ))
    ScaleRGBA=true
   ) {module=ScaleColor001}
  ]
  rendererClasses [&NiagaraSpriteRendererProperties]
  renderers [{rendererClass=&NiagaraSpriteRendererProperties}]
 }
 ConfettiLow {
  emitterUpdateScript [
   /SetVariables(Emitter.BoolDeathEvent:NiagaraBool=&self:ScratchDynamicInput_18(Particle Attribute Reader:NiagaraDataInterfaceParticleRead=DataInterface() {propertyValues {EmitterBinding {bindingMode=Other, emitterName=CrossLama}}})) {isSetParameters=true, module=SetVariables_#8}
   &EmitterState(Life Cycle Mode=::NewEnumerator1, Loop Delay=0.5, Loop Duration=$Engine.DeltaTime, Recalculate Duration Each Loop=true, UseLoopDelay=false)
   &SpawnByTrigger(
    Spawn Count:NiagaraInt32=&RandomRangeInt(Minimum=45, Maximum=55)
    Trigger:NiagaraBool=$Emitter.BoolDeathEvent
    TriggerIntervalSwitch=false
   )
   &SpawnByTrigger(
    Spawn Count:NiagaraInt32=&RandomRangeInt(Minimum=25, Maximum=33)
    Spawn Group=1
    Trigger:NiagaraBool=$Emitter.BoolDeathEvent
    TriggerIntervalSwitch=false
   ) {module=SpawnByTrigger001}
  ]
  particleSpawnScript [
   &InitializeParticle(
    Color=&SelectLinearColorFromArray(
     Array Sampling Mode=::NewEnumerator0
     Direct Array Index=&ReturnExecIndex()
     Color Selection Array=DataInterface() {propertyValues {ColorData [<rgba 0.035 0.2349 1.599 1>, <rgba 0.4453 0.0491 2.2448 1>, <rgba 1.175 0.4351 3.1406 1>, <rgba 0 3.2318 5 1>]}}
     Random Seed=0
    )
    Lifetime Mode=::NewEnumerator1
    Mesh Scale=&Multiply_VectorByFloat(
     Vector=&RandomRangeVector(Minimum=<1 1 1>, Maximum=<1.05 1.05 3>)
     Float=&RandomRangeFloat(Minimum=0.35, Maximum=0.5)
    )
    Mesh Scale Mode=::NewEnumerator4
    Sprite Size=&Multiply_Vector2D_ByFloat(Float=&RandomRangeFloat(Minimum=5, Maximum=10), Vector2D=<1 1.5>)
    Sprite Size Mode=::NewEnumerator4
    UsePositionOffset=false
    Write Material Random=true
   ) {module=InitializeParticle001}
   &SocketLocation(
    Orientation Sampling=::NewEnumerator1
    Position Sampling=::NewEnumerator1
    SamplePosOffset=true
    Sampled Position Coordinate Space:ENiagaraCoordinateSpace=::Local
    Sampled Position Offset=&RandomVector(Vector Scale=5)
    Socket Index=4
    Socket Reader=$System.Socket Reader
    Socket Sampling Mode=::NewEnumerator3
    VelocitySpeedLimit=false
   )
   /SetVariables(Particles.Position:Vector3f=$Output.SocketLocation.SampledPosition) {isSetParameters=true, module=SetVariables_#9}
   &InitialMeshOrientation(Use Rotation Vector=false)
   &AddVelocity(
    Module.Use Falloff Distance=false
    Origin Offset=<0 0 -1>
    Override Seed=false
    Velocity Mode=::NewEnumerator1
    Velocity Origin=$Particles.Position
    Velocity Speed=&MaskFloat(
     Mask Source=0
     Mask Parameter=?
     Mask=0
     Masked Float=&RandomRangeFloat(Minimum=5, Maximum=10)
     Passthrough Float=0
    )
   ) {module=AddVelocity001}
   /Niagara/Modules/Update/Forces/V2/CurlNoiseForce(
    Bias Noise=false
    Noise Frequency=10
    Noise Strength=&RandomRangeFloat(Minimum=25, Maximum=50)
    Pan Noise=true
    Pan Noise Field=<0.2016 0 0.5008>
   ) {module=CurlNoiseForce001}
   &AddVelocity(
    Cone Axis=&Multiply_VectorByFloat(Vector=$System.FN_Contrail_Setup.Direction Vector, Float=-1)
    Module.UseConeFalloff=false
    Module.UseVelDistribution=false
    Override Seed=false
    Velocity Mode=::NewEnumerator2
    Velocity Speed=&MaskFloat(
     Mask Source=0
     Mask Parameter=?
     Mask=0
     Masked Float=&RandomRangeFloat(Minimum=350, Maximum=888)
     Passthrough Float=&RandomRangeFloat(Minimum=550, Maximum=1000)
    )
   ) {module=AddVelocity003}
   &InheritVelocity(
    Inherited Velocity Amount Scale=<0 0 -1>
    Inherited Velocity Speed Limit:NiagaraFloat=&MaskFloat(Mask Source=0, Mask Parameter=?, Mask=0, Masked Float=300, Passthrough Float=400)
    Speed Limit=true
    UseSpeedThreshold=false
    Velocity Destination Space:ENiagaraCoordinateSpace=::World
   )
   &ApplyInitialForces()
   &Color(Color=&MultiplyLinearColor(
     Linear Color=$Particles.Color
     Scale Factor=&MakeVector4_FromLinearColor(LinearColor=&FNTimeOfDayColor(
       Night Color=<rgba 0.2706 0.2383 0.4948 1>
       Morning Color=<rgba 1 0.7262 0.6963 1>
       Day Color=<rgba 1 0.9315 0.9012 1>
       Evening Color=<rgba 1 0.9069 0.7687 1>
      ))
    ))
  ]
  particleUpdateScript [
   &ParticleState()
   &SpriteRotationRate(Rotation Rate=&Multiply_Float(
     A=&SelectFloatFromArray(
      Array Sampling Mode=::NewEnumerator0
      Direct Array Index=?
      Float Selection Array=DataInterface() {propertyValues {FloatData [-300,300]}}
      Random Seed=0
     )
     B=&RandomRangeFloat(Minimum=-5, Maximum=5)
    ))
   &ScaleSpriteSize(Scale Factor=&Vector2DFromFloat(Value=&FloatFromCurve(
      FloatCurve=DataInterface() {propertyValues {Curve {keys [
          {arriveTangent=-0.4596, interpMode=RCIM_Cubic, leaveTangent=-0.4596, tangentMode=RCTM_User, time=0.8, value=1}
          {arriveTangent=-4.9031, interpMode=RCIM_Cubic, leaveTangent=-4.9031, tangentMode=RCTM_User, time=1, value=0}
         ]}}}
      Scale Curve=&RandomRangeFloat(Minimum=1, Maximum=1, Randomness Mode=::NewEnumerator2, Random Seed=?)
     ))) {module=ScaleSpriteSize001}
   &Drag(
    Drag=&MaskFloat(
     Mask Source=0
     Mask Parameter=?
     Mask=0
     Masked Float=50
     Passthrough Float=&RandomRangeFloat(Minimum=7, Maximum=9)
    )
    Use Linear Drag=true
    Use Rotational Drag=true
   ) {module=Drag002}
   /Niagara/Modules/Update/Forces/CurlNoiseForce(
    Noise Frequency=50
    Noise Strength=&MaskFloat(
     Mask Source=0
     Mask Parameter=?
     Mask=0
     Masked Float=&RandomRangeFloat(Minimum=100, Maximum=150)
     Passthrough Float=0
    )
    Pan Noise=false
   )
   &AerodynamicDrag(
    Aerodynamic Drag=0.2
    Drag Shape Curve=DataInterface() {propertyValues {Curve {keys [
        {interpMode=RCIM_Cubic, tangentMode=RCTM_User, time=0, value=0.1}
        {arriveTangent=2.0799, interpMode=RCIM_Cubic, leaveTangent=2.0799, tangentMode=RCTM_User, time=1, value=1}
       ]}}}
    Ignore Mass=true
    Lift Shape Curve=DataInterface() {propertyValues {Curve {keys [
        {interpMode=RCIM_Cubic, tangentMode=RCTM_User, time=0, value=0}
        {interpMode=RCIM_Cubic, tangentMode=RCTM_User, time=0.5, value=1}
        {tangentMode=RCTM_User, time=1, value=0}
       ]}}}
    Surface Normal Axis:Vector3f=<0 0 1>
    Use Linear Drag=true
    Use Rotational Drag=true
   )
   &GravityForce(Gravity=<0 0 980>)
   &SolveForcesAndVelocity(Clamp Velocity=false, Limit Acceleration=false)
   &ScaleSpriteSizeBySpeed(
    Max Scale Factor=<1 1>
    Min Scale Factor=/Niagara/DynamicInputs/Helpers/MaskVector2d(
     Mask Source:NiagaraInt32=0
     Mask Parameter:NiagaraInt32=?
     Mask:NiagaraInt32=0
     Masked Vector2D:Vector2f=<1 1>
     Passthrough Vector2D:Vector2f=<0 0>
    )
    Velocity Threshold=20
   )
   &AlignSpriteToMeshOrientation(SetAlignment=true, SetFacing=true)
   /SetVariables(Particles.SubImageIndex=&FloatFromCurve(FloatCurve=DataInterface() {propertyValues {Curve {keys [<at 0 0>,<at 1 1>]}}}, Scale Curve=36)) {isSetParameters=true, module=SetVariables_#10}
   &KillParticles(Kill Particles=&SetBoolByFloatComparison(A=&VectorLength(Vector=$Particles.Velocity), B=1, Comparison Type=::NewEnumerator6))
   &ScaleColor(
    Scale Mode=1
    Scale RGBA=&MakeVector4_FromLinearColor(LinearColor=&FNTimeOfDayColor(
      Night Color=&ColorFromCurve(ColorCurve=DataInterface() {propertyValues {
         AlphaCurve {keys [<at 0 1 RCIM_Cubic>]}
         BlueCurve {keys [<at 0 1 RCIM_Cubic>,<at 0.4559 1>,<at 0.4752 250>,<at 0.4979 1>,<at 1 1 RCIM_Cubic>]}
         GreenCurve {keys [<at 0 1 RCIM_Cubic>,<at 0.4559 1>,<at 0.4752 250>,<at 0.4979 1>,<at 1 1 RCIM_Cubic>]}
         RedCurve {keys [<at 0 1 RCIM_Cubic>,<at 0.4559 1>,<at 0.4752 250>,<at 0.4979 1>,<at 1 1 RCIM_Cubic>]}
        }})
      Morning Color=&ColorFromCurve(ColorCurve=DataInterface() {propertyValues {
         AlphaCurve {keys [<at 0 1 RCIM_Cubic>]}
         BlueCurve {keys [<at 0 1 RCIM_Cubic>,<at 0.4559 1>,<at 0.4752 250>,<at 0.4979 1>,<at 1 1 RCIM_Cubic>]}
         GreenCurve {keys [<at 0 1 RCIM_Cubic>,<at 0.4559 1>,<at 0.4752 250>,<at 0.4979 1>,<at 1 1 RCIM_Cubic>]}
         RedCurve {keys [<at 0 1 RCIM_Cubic>,<at 0.4559 1>,<at 0.4752 250>,<at 0.4979 1>,<at 1 1 RCIM_Cubic>]}
        }})
      Day Color=&ColorFromCurve(ColorCurve=DataInterface() {propertyValues {
         AlphaCurve {keys [<at 0 1 RCIM_Cubic>]}
         BlueCurve {keys [<at 0 1 RCIM_Cubic>,<at 0.4559 1>,<at 0.4752 250>,<at 0.4979 1>,<at 1 1 RCIM_Cubic>]}
         GreenCurve {keys [<at 0 1 RCIM_Cubic>,<at 0.4559 1>,<at 0.4752 250>,<at 0.4979 1>,<at 1 1 RCIM_Cubic>]}
         RedCurve {keys [<at 0 1 RCIM_Cubic>,<at 0.4559 1>,<at 0.4752 250>,<at 0.4979 1>,<at 1 1 RCIM_Cubic>]}
        }})
      Evening Color=&ColorFromCurve(ColorCurve=DataInterface() {propertyValues {
         AlphaCurve {keys [<at 0 1 RCIM_Cubic>]}
         BlueCurve {keys [<at 0 1 RCIM_Cubic>,<at 0.4559 1>,<at 0.4752 250>,<at 0.4979 1>,<at 1 1 RCIM_Cubic>]}
         GreenCurve {keys [<at 0 1 RCIM_Cubic>,<at 0.4559 1>,<at 0.4752 250>,<at 0.4979 1>,<at 1 1 RCIM_Cubic>]}
         RedCurve {keys [<at 0 1 RCIM_Cubic>,<at 0.4559 1>,<at 0.4752 250>,<at 0.4979 1>,<at 1 1 RCIM_Cubic>]}
        }})
     ))
    ScaleRGBA=true
   ) {module=ScaleColor001}
  ]
  rendererClasses [&NiagaraSpriteRendererProperties]
  renderers [{rendererClass=&NiagaraSpriteRendererProperties}]
 }
 ConfettiMain {
  simTarget=GPUComputeSim
  emitterUpdateScript [
   /SetVariables(Emitter.BoolDeathEvent:NiagaraBool=&self:ScratchDynamicInput_14(Particle Attribute Reader:NiagaraDataInterfaceParticleRead=DataInterface() {propertyValues {EmitterBinding {bindingMode=Other, emitterName=CrossLama}}})) {isSetParameters=true}
   &EmitterState(Life Cycle Mode=::NewEnumerator1, Loop Delay=0.5, Loop Duration=$Engine.DeltaTime, Recalculate Duration Each Loop=true, UseLoopDelay=false)
   &SpawnByTrigger(
    Spawn Count:NiagaraInt32=&RandomRangeInt(Minimum=150, Maximum=177)
    Trigger:NiagaraBool=$Emitter.BoolDeathEvent
    TriggerIntervalSwitch=false
   )
   &SpawnByTrigger(
    Spawn Count:NiagaraInt32=&RandomRangeInt(Minimum=150, Maximum=177)
    Spawn Group=1
    Trigger:NiagaraBool=$Emitter.BoolDeathEvent
    TriggerIntervalSwitch=false
   ) {module=SpawnByTrigger001}
  ]
  particleSpawnScript [
   &InitializeParticle(
    Color=&SelectLinearColorFromArray(
     Array Sampling Mode=::NewEnumerator0
     Direct Array Index=&ReturnExecIndex()
     Color Selection Array=DataInterface() {propertyValues {ColorData [<rgba 0.035 0.2349 1.599 1>, <rgba 0.4453 0.0491 2.2448 1>, <rgba 1.175 0.4351 3.1406 1>, <rgba 0 3.2318 5 1>]}}
     Random Seed=0
    )
    Lifetime Max=1
    Lifetime Min:NiagaraFloat=&MaskFloat(Mask Source=0, Mask Parameter=?, Mask=0, Masked Float=0.7, Passthrough Float=0.5)
    Lifetime Mode=::NewEnumerator1
    Mesh Scale=&Multiply_VectorByFloat(
     Vector=&RandomRangeVector(Minimum=<1 1 1>, Maximum=<1.05 1.05 3>)
     Float=&RandomRangeFloat(Minimum=0.35, Maximum=0.5)
    )
    Mesh Scale Mode=::NewEnumerator4
    Sprite Size=&Multiply_Vector2D_ByFloat(Float=&RandomRangeFloat(Minimum=5, Maximum=10), Vector2D=<1 1.5>)
    Sprite Size Mode=::NewEnumerator4
    UsePositionOffset=false
    Write Material Random=true
   ) {module=InitializeParticle001}
   &SocketLocation(
    Orientation Sampling=::NewEnumerator1
    Position Sampling=::NewEnumerator1
    SamplePosOffset=true
    Sampled Position Coordinate Space:ENiagaraCoordinateSpace=::Local
    Sampled Position Offset=&MaskVector(
     Mask Source=0
     Mask Parameter=?
     Mask=0
     Masked Vector=&RandomVector(Vector Scale=5)
     Passthrough Vector=<0 0 0>
    )
    Socket Index=4
    Socket Reader=$System.Socket Reader
    Socket Sampling Mode=::NewEnumerator3
    VelocitySpeedLimit=false
   )
   /SetVariables(Particles.Position:Vector3f=$Output.SocketLocation.SampledPosition) {isSetParameters=true}
   &InitialMeshOrientation(Use Rotation Vector=false)
   &AddVelocity(
    Cone Axis=&Multiply_VectorByFloat(Vector=$System.FN_Contrail_Setup.Direction Vector, Float=-1)
    Module.UseConeFalloff=false
    Module.UseVelDistribution=false
    Override Seed=false
    Velocity Mode=::NewEnumerator2
    Velocity Speed=&MaskFloat(
     Mask Source=0
     Mask Parameter=?
     Mask=0
     Masked Float=&RandomRangeFloat(Minimum=550, Maximum=1000)
     Passthrough Float=&RandomRangeFloat(Minimum=333, Maximum=555)
    )
   ) {module=AddVelocity003}
   /Niagara/Modules/Update/Forces/V2/CurlNoiseForce(
    Bias Noise=false
    Noise Frequency=&MaskFloat(Mask Source=0, Mask Parameter=?, Mask=0, Masked Float=10, Passthrough Float=0)
    Noise Strength=&MaskFloat(
     Mask Source=0
     Mask Parameter=?
     Mask=0
     Masked Float=&RandomRangeFloat(Minimum=25, Maximum=50)
     Passthrough Float=0
    )
    Pan Noise=true
    Pan Noise Field=<0.2016 0 0.5008>
   ) {module=CurlNoiseForce001}
   &InheritVelocity(
    Inherited Velocity Amount Scale=<0 0 -1>
    Inherited Velocity Speed Limit=400
    Speed Limit=true
    UseSpeedThreshold=false
    Velocity Destination Space:ENiagaraCoordinateSpace=::World
   )
   &ApplyInitialForces()
   &Color(Color=&MultiplyLinearColor(
     Linear Color=$Particles.Color
     Scale Factor=&MakeVector4_FromLinearColor(LinearColor=&FNTimeOfDayColor(
       Night Color=<rgba 0.2706 0.2383 0.4948 1>
       Morning Color=<rgba 1 0.7262 0.6963 1>
       Day Color=<rgba 1 0.9315 0.9012 1>
       Evening Color=<rgba 1 0.9069 0.7687 1>
      ))
    ))
   &SpriteFacingAndAlignment(Sprite Facing=$Particles.ShapeLocation.ShapeNormal, WriteAlignment=false, WriteFacing=true)
  ]
  particleUpdateScript [
   &ParticleState()
   &SpriteRotationRate(Rotation Rate=&Multiply_Float(
     A=&SelectFloatFromArray(
      Array Sampling Mode=::NewEnumerator0
      Direct Array Index=?
      Float Selection Array=DataInterface() {propertyValues {FloatData [-300,300]}}
      Random Seed=0
     )
     B=&RandomRangeFloat(Minimum=-5, Maximum=5)
    ))
   &ScaleSpriteSize(Scale Factor=&Vector2DFromFloat(Value=&FloatFromCurve(
      FloatCurve=DataInterface() {propertyValues {Curve {keys [
          {arriveTangent=-0.4596, interpMode=RCIM_Cubic, leaveTangent=-0.4596, tangentMode=RCTM_User, time=0.8, value=1}
          {arriveTangent=-4.9031, interpMode=RCIM_Cubic, leaveTangent=-4.9031, tangentMode=RCTM_User, time=1, value=0}
         ]}}}
      Scale Curve=&RandomRangeFloat(Minimum=1, Maximum=1, Randomness Mode=::NewEnumerator2, Random Seed=?)
     ))) {module=ScaleSpriteSize001}
   &Drag(
    Drag=&MaskFloat(
     Mask Source=0
     Mask Parameter=?
     Mask=0
     Masked Float=&RandomRangeFloat(Minimum=7, Maximum=9)
     Passthrough Float=&RandomRangeFloat(Minimum=3, Maximum=5)
    )
    Use Linear Drag=true
    Use Rotational Drag=true
   ) {module=Drag001}
   &AerodynamicDrag(
    Aerodynamic Drag=0.2
    Drag Shape Curve=DataInterface() {propertyValues {Curve {keys [
        {interpMode=RCIM_Cubic, tangentMode=RCTM_User, time=0, value=0.1}
        {arriveTangent=2.0799, interpMode=RCIM_Cubic, leaveTangent=2.0799, tangentMode=RCTM_User, time=1, value=1}
       ]}}}
    Ignore Mass=true
    Lift Shape Curve=DataInterface() {propertyValues {Curve {keys [
        {interpMode=RCIM_Cubic, tangentMode=RCTM_User, time=0, value=0}
        {interpMode=RCIM_Cubic, tangentMode=RCTM_User, time=0.5, value=1}
        {tangentMode=RCTM_User, time=1, value=0}
       ]}}}
    Surface Normal Axis:Vector3f=<0 0 1>
    Use Linear Drag=true
    Use Rotational Drag=true
   )
   &SolveForcesAndVelocity(Clamp Velocity=false, Limit Acceleration=false)
   &ScaleSpriteSizeBySpeed(
    Max Scale Factor=<1 1>
    Min Scale Factor=&MaskVector2d(
     Mask Source:NiagaraInt32=0
     Mask Parameter:NiagaraInt32=?
     Mask:NiagaraInt32=0
     Masked Vector2D:Vector2f=<0 0>
     Passthrough Vector2D:Vector2f=<1 1>
    )
    Velocity Threshold:NiagaraFloat=&MaskFloat(Mask Source=0, Mask Parameter=?, Mask=0, Masked Float=20, Passthrough Float=0)
   )
   &AlignSpriteToMeshOrientation(SetAlignment=true, SetFacing=true)
   /SetVariables(Particles.SubImageIndex=&FloatFromCurve(FloatCurve=DataInterface() {propertyValues {Curve {keys [<at 0 0>,<at 1 1>]}}}, Scale Curve=36)) {isSetParameters=true, module=SetVariables_#10}
   &KillParticles(Kill Particles=&SetBoolByFloatComparison(A=&VectorLength(Vector=$Particles.Velocity), B=1, Comparison Type=::NewEnumerator6))
   &ScaleColor(
    Scale Mode=1
    Scale RGBA=&MakeVector4_FromLinearColor(LinearColor=&FNTimeOfDayColor(
      Night Color=&ColorFromCurve(ColorCurve=DataInterface() {propertyValues {
         AlphaCurve {keys [<at 0 1 RCIM_Cubic>]}
         BlueCurve {keys [<at 0 1 RCIM_Cubic>,<at 0.4559 1>,<at 0.4752 250>,<at 0.4979 1>,<at 1 1 RCIM_Cubic>]}
         GreenCurve {keys [<at 0 1 RCIM_Cubic>,<at 0.4559 1>,<at 0.4752 250>,<at 0.4979 1>,<at 1 1 RCIM_Cubic>]}
         RedCurve {keys [<at 0 1 RCIM_Cubic>,<at 0.4559 1>,<at 0.4752 250>,<at 0.4979 1>,<at 1 1 RCIM_Cubic>]}
        }})
      Morning Color=&ColorFromCurve(ColorCurve=DataInterface() {propertyValues {
         AlphaCurve {keys [<at 0 1 RCIM_Cubic>]}
         BlueCurve {keys [<at 0 1 RCIM_Cubic>,<at 0.4559 1>,<at 0.4752 250>,<at 0.4979 1>,<at 1 1 RCIM_Cubic>]}
         GreenCurve {keys [<at 0 1 RCIM_Cubic>,<at 0.4559 1>,<at 0.4752 250>,<at 0.4979 1>,<at 1 1 RCIM_Cubic>]}
         RedCurve {keys [<at 0 1 RCIM_Cubic>,<at 0.4559 1>,<at 0.4752 250>,<at 0.4979 1>,<at 1 1 RCIM_Cubic>]}
        }})
      Day Color=&ColorFromCurve(ColorCurve=DataInterface() {propertyValues {
         AlphaCurve {keys [<at 0 1 RCIM_Cubic>]}
         BlueCurve {keys [<at 0 1 RCIM_Cubic>,<at 0.4559 1>,<at 0.4752 250>,<at 0.4979 1>,<at 1 1 RCIM_Cubic>]}
         GreenCurve {keys [<at 0 1 RCIM_Cubic>,<at 0.4559 1>,<at 0.4752 250>,<at 0.4979 1>,<at 1 1 RCIM_Cubic>]}
         RedCurve {keys [<at 0 1 RCIM_Cubic>,<at 0.4559 1>,<at 0.4752 250>,<at 0.4979 1>,<at 1 1 RCIM_Cubic>]}
        }})
      Evening Color=&ColorFromCurve(ColorCurve=DataInterface() {propertyValues {
         AlphaCurve {keys [<at 0 1 RCIM_Cubic>]}
         BlueCurve {keys [<at 0 1 RCIM_Cubic>,<at 0.4559 1>,<at 0.4752 250>,<at 0.4979 1>,<at 1 1 RCIM_Cubic>]}
         GreenCurve {keys [<at 0 1 RCIM_Cubic>,<at 0.4559 1>,<at 0.4752 250>,<at 0.4979 1>,<at 1 1 RCIM_Cubic>]}
         RedCurve {keys [<at 0 1 RCIM_Cubic>,<at 0.4559 1>,<at 0.4752 250>,<at 0.4979 1>,<at 1 1 RCIM_Cubic>]}
        }})
     ))
    ScaleRGBA=true
   ) {module=ScaleColor001}
  ]
  rendererClasses [&NiagaraSpriteRendererProperties]
  renderers [{rendererClass=&NiagaraSpriteRendererProperties}]
 }
 CrossLama {
  emitterUpdateScript [
   &EmitterState(Life Cycle Mode=::NewEnumerator1, Loop Delay=1, Loop Duration=0.5, UseLoopDelay=true)
   &SpawnRate(
    SpawnRate=&FN_FrontEndFloatModifier(
     Gameplay=&Lerp_Float(A=0, B=1, Alpha=$System.Dive)
     Front End=&Lerp_Float(A=0, B=1, Alpha=$System.Dive)
    )
    Use Spawn Probability=true
   )
  ]
  particleSpawnScript [
   &InitializeParticle(
    Color Mode=::NewEnumerator0
    Lifetime=&FN_FrontEndFloatModifier(Gameplay=0.7, Front End=0.8)
    Mesh Scale Mode=::NewEnumerator3
    Mesh Uniform Scale=&FN_FrontEndFloatModifier(Gameplay=1.1, Front End=0.5)
    UsePositionOffset=false
    Write Material Random=true
   ) {module=InitializeParticle001}
   &InitialMeshOrientation(
    Mesh Axis To Orient=<0 0 1>
    Mesh Orientation Mode=::NewEnumerator2
    Orientation Vector=<0 0 -1>
    Rotation=&RandomRangeVector(Minimum=<0 0 0>, Maximum=<0.25 0.25 0.25>)
    Use Rotation Vector=true
   ) {module=InitialMeshOrientation001}
   &SystemLocation(MaskSpawnGroup=false, Offset=&RandomRangeVector(Minimum=<-20 -20 -1000>, Maximum=<20 20 -1000>))
  ]
  particleUpdateScript [
   &ParticleState()
   &SocketLocation(
    Orientation Sampling=::NewEnumerator1
    Position Sampling=::NewEnumerator1
    SamplePosOffset=false
    Socket Index=4
    Socket Reader=$System.Socket Reader
    Socket Sampling Mode=::NewEnumerator3
    VelocitySpeedLimit=false
   )
   &ScaleMeshSize(Scale Factor=&VectorFromCurve(VectorCurve=DataInterface() {propertyValues {
       XCurve {keys [
         {arriveTangent=4.5495, interpMode=RCIM_Cubic, leaveTangent=4.5495, tangentMode=RCTM_User, time=0, value=0}
         {arriveTangent=0.2449, interpMode=RCIM_Cubic, leaveTangent=0.2449, tangentMode=RCTM_User, time=0.496, value=1.1594}
         {arriveTangent=-0.0639, interpMode=RCIM_Cubic, leaveTangent=-2.0862, tangentMode=RCTM_Break, time=0.92, value=1.1743}
         {arriveTangent=-4.1056, interpMode=RCIM_Cubic, leaveTangent=-4.1056, tangentMode=RCTM_User, time=1, value=0}
        ]}
       YCurve {keys [
         {arriveTangent=4.5495, interpMode=RCIM_Cubic, leaveTangent=4.5495, tangentMode=RCTM_User, time=0, value=0}
         {arriveTangent=0.2449, interpMode=RCIM_Cubic, leaveTangent=0.2449, tangentMode=RCTM_User, time=0.496, value=1.1594}
         {arriveTangent=-0.0639, interpMode=RCIM_Cubic, leaveTangent=-2.0862, tangentMode=RCTM_Break, time=0.92, value=1.1743}
         {arriveTangent=-4.1056, interpMode=RCIM_Cubic, leaveTangent=-4.1056, tangentMode=RCTM_User, time=1, value=0}
        ]}
       ZCurve {keys [
         {arriveTangent=4.5495, interpMode=RCIM_Cubic, leaveTangent=4.5495, tangentMode=RCTM_User, time=0, value=0}
         {arriveTangent=0.2449, interpMode=RCIM_Cubic, leaveTangent=0.2449, tangentMode=RCTM_User, time=0.496, value=1.1594}
         {arriveTangent=-0.0639, interpMode=RCIM_Cubic, leaveTangent=-2.0862, tangentMode=RCTM_Break, time=0.92, value=1.1743}
         {arriveTangent=-4.1056, interpMode=RCIM_Cubic, leaveTangent=-4.1056, tangentMode=RCTM_User, time=1, value=0}
        ]}
      }})) {module=ScaleMeshSize001}
   /SetVariables(Particles.ImpactPosition:NiagaraPosition=&FN_FrontEndVector3Modifier(
     Gameplay=&AddVectorToPosition(Position=$Output.SocketLocation.SampledPosition, Vector=<0 0 -100>)
     Front End=&AddVectorToPosition(Position=$Output.SocketLocation.SampledPosition, Vector=<0 0 20>)
    )) {isSetParameters=true}
   /SetVariables(Particles.Position=&Lerp_Vector(
     A=Linked() {linkedVariable {name=Particles.Position, type=NiagaraPosition}}
     B=Linked() {linkedVariable {name=Particles.ImpactPosition, type=NiagaraPosition}}
     Alpha=&FloatFromCurve(
      FloatCurve=DataInterface() {propertyValues {Curve {
         defaultValue=0
         keys [
          {interpMode=RCIM_Cubic, leaveTangent=0.1666, tangentMode=RCTM_Break, time=0, value=0}
          {arriveTangent=1.6754, interpMode=RCIM_Cubic, tangentMode=RCTM_Break, time=1, value=1}
         ]
        }}}
      Scale Curve=0.2
     )
    )) {isSetParameters=true}
   &ScaleMeshSize(Scale Factor=&VectorFromFloat(Value=&FloatFromCurve(FloatCurve=DataInterface() {propertyValues {Curve {keys [
          {interpMode=RCIM_Cubic, tangentMode=RCTM_User, time=0, value=0}
          {arriveTangent=0.3483, interpMode=RCIM_Cubic, leaveTangent=0.3483, tangentMode=RCTM_User, time=0.2922, value=1}
          {arriveTangent=-0.3187, interpMode=RCIM_Cubic, leaveTangent=-0.3187, tangentMode=RCTM_Break, time=0.9, value=1}
          {arriveTangent=-0.4322, interpMode=RCIM_Cubic, leaveTangent=-0.4322, tangentMode=RCTM_Break, time=1, value=0}
         ]}}}))) {module=ScaleMeshSize003}
   &MeshRotationForce(Rotation=&RandomRangeVector(
     Minimum=&RandomRangeVector(Minimum=<-0.5 -0.5 -0.5>, Maximum=<-2 -2 -2>)
     Maximum=&RandomRangeVector(Minimum=<-2.5 -2.5 -2.5>, Maximum=<-3 -3 -3>)
    )) {module=MeshRotationForce002}
   &SolveForcesAndVelocity(Clamp Velocity=false, Limit Acceleration=false)
   /SetVariables(Particles.LamaDeath:NiagaraBool=&SetBoolByFloatComparison(A=$Particles.Age, B=&Multiply_Float(A=$Particles.Lifetime, B=0.9), Comparison Type=::NewEnumerator2)) {isSetParameters=true}
  ]
  rendererClasses [&NiagaraMeshRendererProperties]
  renderers [{rendererClass=&NiagaraMeshRendererProperties, rendererIndex=0}]
 }
 LamaPieces {
  emitterUpdateScript [
   /SetVariables(Emitter.BoolDeathEvent:NiagaraBool=&self:ScratchDynamicInput_16(Particle Attribute Reader:NiagaraDataInterfaceParticleRead=DataInterface() {propertyValues {EmitterBinding {bindingMode=Other, emitterName=CrossLama}}})) {isSetParameters=true}
   &EmitterState(Life Cycle Mode=::NewEnumerator1, Loop Delay=0.05, Loop Duration=$Engine.DeltaTime, Recalculate Duration Each Loop=true, UseLoopDelay=false)
   &SpawnByTrigger(Spawn Count=5, Trigger:NiagaraBool=$Emitter.BoolDeathEvent, TriggerIntervalSwitch=false)
  ]
  particleSpawnScript [
   &InitializeParticle(
    Color Mode=::NewEnumerator0
    Lifetime Max=1
    Lifetime Min=0.55
    Lifetime Mode=::NewEnumerator1
    Mesh Renderer Array Visibility Mode=::NewEnumerator1
    Mesh Renderer Info=DataInterface() {propertyValues {MeshRenderer=&self:NiagaraEmitter_16.NiagaraMeshRendererProperties_0}}
    Mesh Renderer Mesh Index:NiagaraInt32=&ReturnParticleIDAttributes()
    Mesh Scale Mode=::NewEnumerator3
    Mesh Uniform Scale=&FN_FrontEndFloatModifier(Gameplay=1.1, Front End=0.5)
    Position Offset=&SelectVectorFromArray(
     Direct Array Index=&ReturnParticleIDAttributes()
     Direct Array Mode=::NewEnumerator1
     Vector Selection Array=$Emitter.Part Position offsets
    )
    UsePositionOffset=true
    Write Material Random=true
   )
   /SetVariables(Particles.MeshIndex=&SelectIntFromArray(
     Direct Array Index=&ReturnExecIndex()
     Int Selection Array=DataInterface() {propertyValues {IntData [0,1,2,3,4]}}
    )) {isSetParameters=true}
   &SocketLocation(
    Orientation Sampling=::NewEnumerator1
    Position Sampling=::NewEnumerator1
    SamplePosOffset=false
    Socket Index=4
    Socket Reader=$System.Socket Reader
    Socket Sampling Mode=::NewEnumerator3
    VelocitySpeedLimit=false
   )
   /SetVariables(Particles.Position:Vector3f=$Output.SocketLocation.SampledPosition) {isSetParameters=true}
   &AddVelocity(
    Cone Angle=180
    Cone Axis=&Multiply_VectorByFloat(Vector=$System.FN_Contrail_Setup.Direction Vector, Float=-1)
    Inner Cone Angle=60
    Module.UseConeFalloff=false
    Module.UseVelDistribution=false
    Override Seed=false
    Velocity Mode=::NewEnumerator2
    Velocity Speed=&FN_FrontEndFloatModifier(
     Gameplay=&RandomRangeFloat(Minimum=700, Maximum=900)
     Front End=&RandomRangeFloat(Minimum=400, Maximum=500)
    )
   ) {module=AddVelocity001}
   &AddVelocity(
    Module.Use Falloff Distance=false
    Origin Offset Coordinate Space=::World
    Override Seed=false
    Velocity Mode=::NewEnumerator1
    Velocity Origin=$Particles.Position
    Velocity Speed=&FN_FrontEndFloatModifier(
     Gameplay=&RandomRangeFloat(Minimum=50, Maximum=100)
     Front End=&RandomRangeFloat(Minimum=10, Maximum=50)
    )
   )
  ]
  particleUpdateScript [
   &ParticleState()
   &UpdateMeshOrientation(
    Rotation Rate=&RandomRangeFloat(Minimum=0.5, Maximum=1)
    Rotation Vector=&RandomVector(Vector Scale=1)
   ) {module=UpdateMeshOrientation001}
   &Drag(Drag=&RandomRangeFloat(Minimum=1, Maximum=2), Use Linear Drag=true, Use Rotational Drag=true) {module=Drag002}
   &GravityForce(Gravity=<0 0 980>)
   &SolveForcesAndVelocity(Clamp Velocity=false, Limit Acceleration=false)
   &ScaleMeshSize(Scale Factor=&VectorFromFloat(Value=&FloatFromCurve(FloatCurve=DataInterface() {propertyValues {Curve {keys [
          {interpMode=RCIM_Cubic, tangentMode=RCTM_Break, time=0.5, value=1}
          {arriveTangent=-7.075, interpMode=RCIM_Cubic, tangentMode=RCTM_Break, time=1, value=0}
         ]}}})))
  ]
  rendererClasses [&NiagaraMeshRendererProperties]
  renderers [{rendererClass=&NiagaraMeshRendererProperties, rendererIndex=0}]
 }
 Lamas {
  emitterUpdateScript [
   &EmitterState(Life Cycle Mode=::NewEnumerator1, Loop Duration=1, UseLoopDelay=false)
   &SpawnRate(
    SpawnRate=&FN_FrontEndFloatModifier(Gameplay=&Lerp_Float(A=2, B=1.5, Alpha=$System.Dive), Front End=2)
    Use Spawn Probability=true
   )
  ]
  particleSpawnScript [
   /SetVariables(Particles.MeshIndex=&SpawnGroup()) {isSetParameters=true}
   &InitializeParticle(
    Lifetime Max=&Lerp_Float(A=3.2, B=2.2, Alpha=$System.Dive)
    Lifetime Min:NiagaraFloat=&Lerp_Float(A=3, B=2, Alpha=$System.Dive)
    Lifetime Mode=::NewEnumerator1
    Mesh Scale Mode=::NewEnumerator3
    Mesh Uniform Scale=&FN_FrontEndFloatModifier(Gameplay=1.1, Front End=0.5)
    Sprite Size Mode=::NewEnumerator1
    UsePositionOffset=false
    Write Material Random=true
   )
   /SetVariables(
    Particles.LamaSpawnHeight:NiagaraFloat=&FN_FrontEndFloatModifier(Gameplay=&Lerp_Float(A=40, B=50, Alpha=$User.ParaGlideLeanAlpha), Front End=20)
    Particles.LamaSpawnRadius:NiagaraFloat=&FN_FrontEndFloatModifier(Gameplay=250, Front End=150)
   ) {isSetParameters=true}
   /SetVariables(Particles.PerSeed:NiagaraFloat=&Add_Float(
     A=&Divide_Float(A=&MakeFloatFromInt(Int=&ReturnParticleIDAttributes()), B=2)
     B=&RandomRangeFloat(Minimum=-0.25, Maximum=0.25)
    )) {isSetParameters=true}
   &InitialMeshOrientation(
    Mesh Axis To Orient=<0 0 1>
    Mesh Orientation Mode=::NewEnumerator2
    Orientation Vector=<0 0 -1>
    Rotation=&RandomRangeVector(Minimum=<0 0 0>, Maximum=<0.25 0.25 0.25>)
    Use Rotation Vector=true
   )
   &SystemLocation(Offset=&AddVectorToPosition(
     Position=$Engine.Emitter.SimulationPosition
     Vector=&AddVectorToPosition(
      Position=&FN_FrontEndVector3Modifier(Gameplay=&Lerp_Vector(A=<0 0 -500>, B=<0 0 -500>, Alpha=$System.Dive), Front End=<0 0 -300>)
      Vector=&MakeVector(
       X=&Sine(
        Normalized Angle=$Particles.PerSeed
        Period=1
        Scale=&Multiply_Float(A=&RandomRangeFloat(Minimum=1, Maximum=1), B=$Particles.LamaSpawnRadius)
        Bias=0
       )
       Y=&Cosine(
        Normalized Angle=$Particles.PerSeed
        Period=1
        Scale=&Multiply_Float(A=&RandomRangeFloat(Minimum=1, Maximum=1), B=$Particles.LamaSpawnRadius)
        Bias=0
       )
       Z=&Multiply_Float(A=$Particles.LamaSpawnHeight, B=&RandomRangeFloat(Minimum=0, Maximum=1))
      )
     )
    )) {module=SystemLocation001}
  ]
  particleUpdateScript [
   &ParticleState()
   &ScaleMeshSize(Scale Factor=&VectorFromCurve(VectorCurve=DataInterface() {propertyValues {
       XCurve {keys [
         {arriveTangent=4.5495, interpMode=RCIM_Cubic, leaveTangent=4.5495, tangentMode=RCTM_User, time=0, value=0}
         {arriveTangent=1.6402, interpMode=RCIM_Cubic, leaveTangent=1.6402, time=0.15, value=1}
         {arriveTangent=-1.4286, interpMode=RCIM_Cubic, leaveTangent=-7.9375, tangentMode=RCTM_Break, time=0.621, value=1.0186}
         {arriveTangent=-6.2546, interpMode=RCIM_Cubic, leaveTangent=-6.2546, tangentMode=RCTM_User, time=0.8, value=0}
        ]}
       YCurve {keys [
         {arriveTangent=4.5495, interpMode=RCIM_Cubic, leaveTangent=4.5495, tangentMode=RCTM_User, time=0, value=0}
         {arriveTangent=1.6402, interpMode=RCIM_Cubic, leaveTangent=1.6402, time=0.15, value=1}
         {arriveTangent=-1.4286, interpMode=RCIM_Cubic, leaveTangent=-7.9375, tangentMode=RCTM_Break, time=0.621, value=1.0186}
         {arriveTangent=-6.2546, interpMode=RCIM_Cubic, leaveTangent=-6.2546, tangentMode=RCTM_User, time=0.8, value=0}
        ]}
       ZCurve {keys [
         {arriveTangent=4.5495, interpMode=RCIM_Cubic, leaveTangent=4.5495, tangentMode=RCTM_User, time=0, value=0}
         {arriveTangent=1.6402, interpMode=RCIM_Cubic, leaveTangent=1.6402, time=0.15, value=1}
         {arriveTangent=-1.4286, interpMode=RCIM_Cubic, leaveTangent=-7.9375, tangentMode=RCTM_Break, time=0.621, value=1.0186}
         {arriveTangent=-6.2546, interpMode=RCIM_Cubic, leaveTangent=-6.2546, tangentMode=RCTM_User, time=0.8, value=0}
        ]}
      }}))
   &MeshRotationForce(Rotation=&Lerp_Vector(
     A=&RandomRangeVector(
      Minimum=&RandomRangeVector(Minimum=<-1 -1 -1>, Maximum=<-3 -3 -3>)
      Maximum=&RandomRangeVector(Minimum=<-2.5 -2.5 -2.5>, Maximum=<3 3 3>)
     )
     B=&RandomRangeVector(
      Minimum=&RandomRangeVector(Minimum=<-1 -1 -1>, Maximum=<-2 -2 -2>)
      Maximum=&RandomRangeVector(Minimum=<-2 -2 -2>, Maximum=<2 2 2>)
     )
     Alpha=$System.Dive
    ))
   &KillParticlesInVolume(Origin Offset:Vector3f=<0 0 -500>, Sphere Radius=200)
   &GravityForce(Gravity=&FN_FrontEndVector3Modifier(
     Gameplay=&Lerp_Vector(A=<0 0 500>, B=<0 0 1500>, Alpha=$System.Dive)
     Front End=&Lerp_Vector(A=<0 0 350>, B=<0 0 1000>, Alpha=$System.Dive)
    ))
   &SolveForcesAndVelocity(Clamp Velocity=false, Limit Acceleration=false)
  ]
  rendererClasses [&NiagaraMeshRendererProperties]
  renderers [{rendererClass=&NiagaraMeshRendererProperties, rendererIndex=0}]
 }
 LittleRibbons {
  emitterUpdateScript [
   /SetVariables(Emitter.BoolDeathEvent:NiagaraBool=&self:ScratchDynamicInput_13(Particle Attribute Reader:NiagaraDataInterfaceParticleRead=DataInterface() {propertyValues {EmitterBinding {bindingMode=Other, emitterName=CrossLama}}})) {isSetParameters=true}
   &EmitterState(Life Cycle Mode=::NewEnumerator1, Loop Duration=$Engine.DeltaTime, Recalculate Duration Each Loop=true, UseLoopDelay=false)
   &SpawnByTrigger(Spawn Count=5, Trigger:NiagaraBool=$Emitter.BoolDeathEvent, TriggerIntervalSwitch=false)
  ]
  particleSpawnScript [
   &InitializeParticle(
    Color=&SelectLinearColorFromArray(
     Array Sampling Mode=::NewEnumerator0
     Direct Array Index=?
     Color Selection Array=DataInterface() {propertyValues {ColorData [<rgba 0.4147 0.134 1.4844 1>, <rgba 0.8337 0.395 3 1>, <rgba 0.0796 0 1.8594 1>, <rgba 0 1.9391 3 1>]}}
     Random Seed=0
    )
    Lifetime=&RandomRangeFloat(Minimum=0.8, Maximum=1.2)
    Mesh Scale Mode=::NewEnumerator3
    Mesh Uniform Scale=&FN_FrontEndFloatModifier(
     Gameplay=&RandomRangeFloat(Minimum=1, Maximum=3)
     Front End=&RandomRangeFloat(Minimum=1, Maximum=3)
    )
    UsePositionOffset=false
    Write Material Random=true
   ) {module=InitializeParticle001}
   &SocketLocation(
    Orientation Sampling=::NewEnumerator1
    SamplePosOffset=false
    Socket Index=4
    Socket Reader=$System.Socket Reader
    Socket Sampling Mode=::NewEnumerator3
    VelocitySpeedLimit=false
   )
   /SetVariables(Particles.Position:Vector3f=$Output.SocketLocation.SampledPosition) {isSetParameters=true}
   &AddVelocity(
    Cone Angle=70
    Cone Axis=<0 0 1>
    Module.UseConeFalloff=true
    Module.UseVelDistribution=true
    Override Seed=false
    Velocity Mode=::NewEnumerator2
    Velocity Speed=&RandomRangeFloat(Minimum=333, Maximum=55)
   )
   &AddVelocity(
    Module.Use Falloff Distance=false
    Origin Offset Coordinate Space=::World
    Override Seed=false
    Velocity Mode=::NewEnumerator1
    Velocity Origin=$Particles.Position
    Velocity Speed=&FN_FrontEndFloatModifier(
     Gameplay=&RandomRangeFloat(Minimum=222, Maximum=333)
     Front End=&RandomRangeFloat(Minimum=10, Maximum=30)
    )
   ) {module=AddVelocity001}
   &InitialMeshOrientation(
    Mesh Orientation Mode=::NewEnumerator2
    Orientation Coordinate Space=::Local
    Orientation Vector=$Particles.Velocity
    Rotation=<0.125 0 0>
    Use Rotation Vector=true
   )
  ]
  particleUpdateScript [
   &ParticleState()
   &GravityForce(Gravity=<0 0 111>)
   &AccelerationForce(
    Acceleration=&Multiply_VectorByFloat(
     Vector=&Lerp_Vector(A=<0 0 99>, B=<0 0 111>, Alpha=$User.ParaGlideLeanAlpha)
     Float=&Lerp_Float(
      A=&FloatFromCurve(
       FloatCurve=DataInterface() {propertyValues {Curve {keys [
           {
            arriveTangent=0.8914
            arriveTangentWeight=1.02
            interpMode=RCIM_Cubic
            leaveTangent=0.8914
            leaveTangentWeight=1.02
            tangentMode=RCTM_User
            tangentWeightMode=RCTWM_WeightedBoth
            time=0
            value=0
           }
           {arriveTangent=1.3544, interpMode=RCIM_Cubic, leaveTangent=1.3544, tangentMode=RCTM_User, time=1, value=1}
          ]}}}
       Scale Curve=15
      )
      B=&FloatFromCurve(
       FloatCurve=DataInterface() {propertyValues {Curve {keys [
           {
            arriveTangent=0.9894
            arriveTangentWeight=1.0012
            interpMode=RCIM_Cubic
            leaveTangent=0.9894
            leaveTangentWeight=1.0012
            tangentMode=RCTM_User
            tangentWeightMode=RCTWM_WeightedBoth
            time=0
            value=0
           }
           {arriveTangent=1.0204, interpMode=RCIM_Cubic, leaveTangent=1.0204, tangentMode=RCTM_User, time=1, value=1}
          ]}}}
       Scale Curve=10
      )
      Alpha=$User.ParaGlideLeanAlpha
     )
    )
    Coordinate Space=::Local
   )
   &Drag(Drag=&RandomRangeFloat(Minimum=5, Maximum=7), Use Linear Drag=true, Use Rotational Drag=false)
   &SolveForcesAndVelocity(Clamp Velocity=false, Limit Acceleration=false)
   &ScaleColor(
    Scale RGB=&VectorFromFloat(Value=&FNTimeOfDayScalar(Night Scale=0.3, Morning Scale=0.333, Day Scale=0.666, Evening Scale=0.333))
    ScaleA=false
    ScaleRGB=true
   )
  ]
  rendererClasses [&NiagaraMeshRendererProperties]
  renderers [{rendererClass=&NiagaraMeshRendererProperties, rendererIndex=0}]
 }
 Ribbons {
  emitterUpdateScript [
   &EmitterState(
    Inactive Response=::NewEnumerator1
    Life Cycle Mode=::NewEnumerator1
    Loop Behavior=::NewEnumerator1
    Loop Duration Mode=::NewEnumerator1
    UseLoopDelay=false
   )
   &SpawnBurst_Instantaneous(Spawn Count=2, Use Spawn Probability=false, UseLoopCountLimit=false)
   &SpawnBurst_Instantaneous(Spawn Count=2, Spawn Group=1, Use Spawn Probability=false, UseLoopCountLimit=false) {module=SpawnBurst_Instantaneous001}
   &SpawnBurst_Instantaneous(Spawn Count=2, Spawn Group=2, Use Spawn Probability=false, UseLoopCountLimit=false) {module=SpawnBurst_Instantaneous002}
   &SpawnBurst_Instantaneous(Spawn Count=2, Spawn Group=3, Use Spawn Probability=false, UseLoopCountLimit=false) {module=SpawnBurst_Instantaneous003}
  ]
  particleSpawnScript [
   &PartitionParticles()
   /SetVariables(Particles.MeshIndex=&MaskIntBySpawnGroup(
     Masked Int=$Particles.PartitionParticles.Partition
     Passthrough Int=&SelectIntFromArray(Direct Array Index=$Particles.PartitionParticles.Partition, Int Selection Array=DataInterface() {propertyValues {IntData [0,0]}})
     Spawn Group Mask=2
    )) {isSetParameters=true}
   /SetVariables(Particles.VisibilityTag=&SpawnGroup(Max Spawn Group=3)) {isSetParameters=true}
   &InitializeParticle(Lifetime Max=1.3, Lifetime Min=0.3, Lifetime Mode=::NewEnumerator1, Position Mode=::NewEnumerator0, Write Material Random=true) {module=InitializeParticle001}
   &InitialMeshOrientation(
    Mesh Axis To Orient=<0 0 1>
    Mesh Orientation Mode=::NewEnumerator2
    Rotation=&MaskVectorBySpawnGroup(
     Masked Vector=&SelectVectorFromArray(
      Direct Array Index=$Particles.PartitionParticles.Partition
      Vector Selection Array=DataInterface() {propertyValues {FloatData [<0.04 0 0.25>, <0.035 0 -0.25>]}}
     )
     Passthrough Vector=<0 0 0>
     Spawn Group Mask=1
    )
    Use Rotation Vector=true
   )
  ]
  particleUpdateScript [
   &ParticleState(Kill Particles When Lifetime Has Elapsed=false, Let Infinitely Lived Particles Die When Emitter Deactivates=true)
   &ScaleMeshSize(Scale Factor=&VectorFromFloat(Value=&FloatFromCurve(
      FloatCurve=DataInterface() {propertyValues {Curve {keys [
          {arriveTangent=1.7685, interpMode=RCIM_Cubic, leaveTangent=1.7685, tangentMode=RCTM_User, time=0.1, value=0}
          {arriveTangent=0.0224, interpMode=RCIM_Cubic, leaveTangent=0.0224, tangentMode=RCTM_User, time=1, value=1}
         ]}}}
      CurveIndex=$Emitter.Age
     ))) {module=ScaleMeshSize001}
   &SocketLocation(
    Orientation Sampling=::NewEnumerator1
    Position Sampling=::NewEnumerator1
    SamplePosOffset=false
    Socket Index=$Particles.PartitionParticles.Partition
    Socket Reader=$System.Socket Reader
    Socket Sampling Mode=::NewEnumerator3
    VelocitySpeedLimit=false
   )
   &UpdateMeshOrientation(
    Facing Direction=&Add_Vector(
     A=$System.AlignmentDirection
     B=&MergeHelper_VectorSwitch(
      0=&Lerp_Vector(
       A=&SelectVectorFromArray(
        Direct Array Index=$Particles.PartitionParticles.Partition
        Vector Selection Array=DataInterface() {propertyValues {FloatData [<0.1 0.01 0>, <-0.05 0.01 0>]}}
       )
       B=&SelectVectorFromArray(
        Direct Array Index=$Particles.PartitionParticles.Partition
        Vector Selection Array=DataInterface() {propertyValues {FloatData [<0.1 0.05 0>, <-0.1 0.05 0>]}}
       )
       Alpha=$System.Dive
      )
      1=&Lerp_Vector(
       A=&SelectVectorFromArray(
        Direct Array Index=$Particles.PartitionParticles.Partition
        Vector Selection Array=DataInterface() {propertyValues {FloatData [<0 0 0>, <0 0 0>]}}
       )
       B=&SelectVectorFromArray(
        Direct Array Index=$Particles.PartitionParticles.Partition
        Vector Selection Array=DataInterface() {propertyValues {FloatData [<-0.1 0.1 0>, <0.1 0.1 0>]}}
       )
       Alpha=$System.Dive
      )
      2=&Lerp_Vector(
       A=&SelectVectorFromArray(
        Direct Array Index=$Particles.PartitionParticles.Partition
        Vector Selection Array=DataInterface() {propertyValues {FloatData [<0.35 -0.02 0>, <-0.35 -0.02 0>]}}
       )
       B=&SelectVectorFromArray(
        Direct Array Index=$Particles.PartitionParticles.Partition
        Vector Selection Array=DataInterface() {propertyValues {FloatData [<0.35 -0.1 0>, <-0.3 -0.1 0>]}}
       )
       Alpha=$System.Dive
      )
      3=&Lerp_Vector(
       A=&SelectVectorFromArray(
        Direct Array Index=$Particles.PartitionParticles.Partition
        Vector Selection Array=DataInterface() {propertyValues {FloatData [<-0.3 0.1 0>, <0.5 0 0>]}}
       )
       B=&SelectVectorFromArray(
        Direct Array Index=$Particles.PartitionParticles.Partition
        Vector Selection Array=DataInterface() {propertyValues {FloatData [<-0.2 0 0>, <0.3 0 0>]}}
       )
       Alpha=$System.Dive
      )
      Index=&SpawnGroup()
      Num=::NewEnumerator2
     )
    )
    Facing Interpolation=false
    Orientation Method=::NewEnumerator1
    Override Quat=false
    Stabilization Mode:ENiagara_OrientVectorCount=::NewEnumerator2
    TransformFacing=false
    TransformSide=false
   ) {module=UpdateMeshOrientation001}
   /SetVariables(Particles.ClothQuaternion:Quat4f=&MultiplyQuaternion(
     Quaternion A=$Particles.MeshOrientation
     Quaternion B=&MakeQuatFromAxisAngle(
      Axis=<0 0 1>
      AngleInDegrees=&MergeHelper_FloatSwitch(0=$System.ClothRotationAngleA, 1=$System.ClothRotationAngleB, 2=?, 3=?, Index=$Particles.PartitionParticles.Partition, Num=::NewEnumerator0)
     )
    )) {isSetParameters=true}
   &DynamicMaterialParameters(
    Index 0 Param 1=&MergeHelper_FloatSwitch(
     0=&MakeCustomFloatFromBool(
      Bool=&SetBoolByFloatComparison(A=&Lerp_Float(A=0, B=1, Alpha=$User.ParaGlideLeanAlpha), B=1, Comparison Type=::NewEnumerator5)
      True Float=&RandomRangeFloat(Minimum=2, Maximum=2.2)
      False Float=3.5
     )
     1=&MakeCustomFloatFromBool(
      Bool=&SetBoolByFloatComparison(A=&Lerp_Float(A=0, B=1, Alpha=$User.ParaGlideLeanAlpha), B=1, Comparison Type=::NewEnumerator5)
      True Float=2.5
      False Float=4
     )
     2=&MakeCustomFloatFromBool(
      Bool=&SetBoolByFloatComparison(A=&Lerp_Float(A=0, B=1, Alpha=$User.ParaGlideLeanAlpha), B=1, Comparison Type=::NewEnumerator5)
      True Float=3
      False Float=5
     )
     3=&MakeCustomFloatFromBool(
      Bool=&SetBoolByFloatComparison(A=&Lerp_Float(A=0, B=1, Alpha=$User.ParaGlideLeanAlpha), B=1, Comparison Type=::NewEnumerator5)
      True Float=3
      False Float=5
     )
     Index=&SpawnGroup()
     Num=::NewEnumerator2
    )
    Index 0 Param 2=&SelectFloatFromArray(Direct Array Index=$Particles.PartitionParticles.Partition, Float Selection Array=DataInterface() {propertyValues {FloatData [0.33,0.333]}})
    Index 0 Param 3=&MergeHelper_FloatSwitch(
     0=&MakeCustomFloatFromBool(
      Bool=&SetBoolByFloatComparison(A=&Lerp_Float(A=0, B=1, Alpha=$User.ParaGlideLeanAlpha), B=1, Comparison Type=::NewEnumerator5)
      True Float=&RandomRangeFloat(Minimum=1.5, Maximum=1.6)
      False Float=&RandomRangeFloat(Minimum=1.2, Maximum=1.3)
     )
     1=&MakeCustomFloatFromBool(
      Bool=&SetBoolByFloatComparison(A=&Lerp_Float(A=0, B=1, Alpha=$User.ParaGlideLeanAlpha), B=1, Comparison Type=::NewEnumerator5)
      True Float=&RandomRangeFloat(Minimum=1.4, Maximum=1.6)
      False Float=&RandomRangeFloat(Minimum=1, Maximum=1.2)
     )
     2=&MakeCustomFloatFromBool(
      Bool=&SetBoolByFloatComparison(A=&Lerp_Float(A=0, B=1, Alpha=$User.ParaGlideLeanAlpha), B=1, Comparison Type=::NewEnumerator5)
      True Float=0.8
      False Float=1
     )
     3=&MakeCustomFloatFromBool(
      Bool=&SetBoolByFloatComparison(A=&Lerp_Float(A=0, B=1, Alpha=$User.ParaGlideLeanAlpha), B=1, Comparison Type=::NewEnumerator5)
      True Float=1.2
      False Float=1
     )
     Index=0
     Num=::NewEnumerator2
    )
    Module.Index0Param0Write=true
    Module.Index0Param1Write=true
    Module.Index0Param2Write=true
    Module.Index0Param3Write=false
   ) {module=DynamicMaterialParameters002}
   /SetVariables(Particles.Position:Vector3f=$Output.SocketLocation.SampledPosition) {isSetParameters=true}
   /SetVariables(Particles.Scale=&Multiply_Vector(
     A=&MergeHelper_VectorSwitch(
      0=<0.1 0.1 0.85>
      1=<0.3 0.3 0.67>
      2=<0.4 0.4 0.47>
      3=&RandomRangeVector(Minimum=<0.3 0.3 0.15>, Maximum=<0.45 0.45 0.25>)
      Index=&SpawnGroup()
      Num=::NewEnumerator2
     )
     B=&MakeVector(X=1, Y=1, Z=&Lerp_Float(A=1, B=1.3, Alpha=$System.Dive))
    )) {isSetParameters=true}
  ]
  rendererClasses [&NiagaraMeshRendererProperties]
  renderers [
   {rendererClass=&NiagaraMeshRendererProperties, rendererIndex=0}
   {rendererClass=&NiagaraMeshRendererProperties, rendererIndex=1}
   {rendererClass=&NiagaraMeshRendererProperties, rendererIndex=2}
   {rendererClass=&NiagaraMeshRendererProperties, rendererIndex=3}
  ]
 }
 Speedlines {
  simTarget=GPUComputeSim
  emitterUpdateScript [
   &EmitterState(Life Cycle Mode=::NewEnumerator1, Loop Duration=1, UseLoopDelay=false)
   &SpawnRate(SpawnRate=&Lerp_Float(A=0, B=50, Alpha=$User.ParaGlideLeanAlpha), Use Spawn Probability=false) {module=SpawnRate001}
  ]
  particleSpawnScript [
   &InitializeParticle(
    Color=&MakeLinearColorFromVectorAndFloat(
     "Vector (RGB)"=<1 1 1>
     "Float (Alpha)"=&Multiply_Float(
      A=&Multiply_Float(
       A=&UniformRangedFloat(Minimum=0.015, Maximum=0.12)
       B=&FN_FrontEndFloatModifier(Gameplay=1, Front End=0.5)
      )
      B=&Add_Float(
       A=1
       B=&Multiply_Float(A=10, B=&Subtract_Float(A=1, B=$Engine.Emitter.SpawnCountScale))
      )
     )
    )
    Lifetime Max=0.5
    Lifetime Min=0.3
    Lifetime Mode=::NewEnumerator1
    Mass Mode=::NewEnumerator1
    Sprite Size=&UniformRangedVector2D(Minimum=<3 4>, Maximum=<5 6>)
    Sprite Size Mode=::NewEnumerator4
    UsePositionOffset=false
    Write Material Random=true
   ) {module=InitializeParticle001}
   &CylinderLocation(
    Cylinder Radius=&UniformRangedFloat(Minimum=50, Maximum=200)
    MaskSpawnGroup=false
    Override Seed=false
    UseOwnerScale=true
   )
   &AddVelocity(Velocity=&FN_FrontEndVector3Modifier(Gameplay=<0 0 1000>, Front End=<0 0 3000>))
   &AddVelocityFromPoint(
    Module.Use Falloff Distance=false
    Velocity Strength=&Multiply_Float(
     A=&FN_FrontEndFloatModifier(Gameplay=1, Front End=2)
     B=&UniformRangedFloat(Minimum=500, Maximum=700)
    )
   )
  ]
  particleUpdateScript [
   &ParticleState()
   &SolveForcesAndVelocity(Clamp Velocity=false, Limit Acceleration=false)
   &ScaleSpriteSize(Scale Factor=&Vector2DFromCurve(
     Vector2Curve=DataInterface() {propertyValues {XCurve {keys [<at 0 0.5 RCIM_Cubic>,<at 0.5 1 RCIM_Cubic>]}, YCurve {keys [<at -0.0111 0.1119 RCIM_Cubic>,<at 0.5 1 RCIM_Cubic>]}}}
     ScaleCurve=&Vector2DFromFloat(Value=&RandomRangeFloat(Minimum=1, Maximum=2))
    ))
   &ScaleSpriteSizeBySpeed(Max Scale Factor=<1 70>, Min Scale Factor=<0.01 0.05>, Velocity Threshold=1)
   &ScaleColor(
    Scale Alpha=&FloatFromCurve(
     FloatCurve=DataInterface() {propertyValues {Curve {keys [
         {arriveTangent=7.861, interpMode=RCIM_Cubic, leaveTangent=7.861, tangentMode=RCTM_User, time=0, value=0}
         {interpMode=RCIM_Cubic, tangentMode=RCTM_Break, time=0.2, value=1}
         {arriveTangent=-2.221, interpMode=RCIM_Cubic, leaveTangent=-2.221, tangentMode=RCTM_User, time=1, value=0}
        ]}}}
     Scale Curve=&FN_FrontEndFloatModifier(Gameplay=&Lerp_Float(A=0, B=0.21, Alpha=$User.ParaGlideLeanAlpha), Front End=0)
    )
    ScaleA=true
    ScaleRGB=true
   )
  ]
  rendererClasses [&NiagaraSpriteRendererProperties]
  renderers [{rendererClass=&NiagaraSpriteRendererProperties}]
 }
}
summary {
 systemName=NS_Contrail_Smashy
 emitters [
  {emitterName=Lamas, rendererClasses [&NiagaraMeshRendererProperties]}
  {emitterName=CrossLama, rendererClasses [&NiagaraMeshRendererProperties]}
  {emitterName=LittleRibbons, rendererClasses [&NiagaraMeshRendererProperties]}
  {emitterName=ConfettiMain, simTarget=GPUComputeSim, rendererClasses [&NiagaraSpriteRendererProperties]}
  {emitterName=ConfettiCenter, rendererClasses [&NiagaraSpriteRendererProperties]}
  {emitterName=Ribbons, rendererClasses [&NiagaraMeshRendererProperties]}
  {emitterName=LamaPieces, rendererClasses [&NiagaraMeshRendererProperties]}
  {emitterName=BurstSmoke, rendererClasses [&NiagaraSpriteRendererProperties, &NiagaraMeshRendererProperties]}
  {emitterName=ConfettiLow, rendererClasses [&NiagaraSpriteRendererProperties]}
  {emitterName=Speedlines, simTarget=GPUComputeSim, rendererClasses [&NiagaraSpriteRendererProperties]}
  {emitterName=ArmConfetti, rendererClasses [&NiagaraSpriteRendererProperties]}
 ]
 userVariables [
  User.bIsFrontEnd {type=NiagaraBool}
  User.bIsFrontEndPreview {type=NiagaraBool}
  User.ParaGlideLeanAlpha {type=NiagaraFloat, defaultValue=0, description=""}
 ]
}
